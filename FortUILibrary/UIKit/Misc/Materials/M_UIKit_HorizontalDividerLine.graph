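# UI material for a horizontal divider line in the UIKit library, using translucent blending.
# It exposes a ColorBase vector parameter, read as RGB and white by default, and an Opacity scalar parameter that defaults to fully opaque.
# Each parameter passes through a reroute, with no textures, masks or math shaping the line.
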Material /FortUILibrary/UIKit/Misc/Materials/M_UIKit_HorizontalDividerLine
graph {
 properties {
  bUsedWithStaticMesh=false
  blendMode=BLEND_Translucent
  materialDomain=MD_UI
  pixelDepthOffsetMode=PDOM_Legacy
  usedShadingModels=Default Lit
 }
 nodes [
  Reroute(None=ScalarParameter() {defaultValue=1, group=Opacity, parameterName=Opacity})
  Reroute(None=VectorParameter() {group=Color, parameterName=ColorBase, defaultValue=<rgba 1 1 1 1>}.RGB)
 ]
}
