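# Translucent UI material that draws a segmented ring for the emote wheel, with one highlighted segment and an outer glow.
# UVs rotate by HighlightSegment, so a floor/ceil test on the radial angle can pick that segment and lerp its color and thickness.
# Circle distance fields produce anti-aliased ring and smoothstep glow masks for both states.
# MF_UI_RadialSegments supplies the segment gaps and their glow falloff, driven by the Segments and SegmentSize parameters.
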Material /FortUILibrary/UIKit/Misc/Materials/M_UI_EmoteWheel_Position
graph {
 properties {
  bAutomaticallySetUsageInEditor=false
  bUsedWithStaticMesh=false
  blendMode=BLEND_Translucent
  materialDomain=MD_UI
  pixelDepthOffsetMode=PDOM_Legacy
  usedShadingModels=Default Lit
 }
 nodes [
  n1 = VectorParameter() {group=Color, parameterName=OuterGlowColor, sortPriority=0, defaultValue=<rgba 1 1 1 0.5>}
  n2 = VectorParameter() {group=Color, parameterName=RingColor, sortPriority=0, defaultValue=<rgba 0 0.0412 1 0.3>}
  n3 = NamedRerouteUsage()
  n4 = VectorParameter() {group=Color, parameterName=HighlightColor, sortPriority=1, defaultValue=<rgba 1 0 0.9063 0.5>}
  n5 = NamedRerouteUsage()
  n6 = NamedRerouteUsage()
  n7 = Multiply(
   A=MaterialFunctionCall!(Vector or UVs=NamedRerouteUsage()) {materialFunction=/FortUILibrary/MaterialFunctions/Basics/MF_UI_CheaperVectorToRadial}.Vector Converted to Angle
   B=@n6
  )
  n8 = Floor(None=Divide(A=Ceil(None=@n7), B=@n6))
  n9 = LinearInterpolate(A=Multiply(A=AppendVector(A=@n2.RGB, B=@n2.RGB), B=@n3), B=Multiply(A=AppendVector(A=@n4.RGB, B=@n4.RGB), B=@n5), Alpha=@n8)
  n10 = NamedRerouteUsage()
  n11 = NamedRerouteUsage()
  LinearInterpolate(
   A=@n1.RGB
   B=ComponentMask(None=@n9) {b=true, g=true, r=true}
   Alpha=Multiply(A=LinearInterpolate(A=@n3, B=@n5, Alpha=@n8), B=Add(A=@n10, B=@n11))
  ) {constAlpha=1}
  n12 = LinearInterpolate(A=NamedRerouteUsage(), B=NamedRerouteUsage(), Alpha=@n8)
  Min(A=Add(A=Multiply(A=@n12, B=@n1.A), B=ComponentMask(None=@n9) {a=true}), B=Add(A=@n10, B=Multiply(A=@n11, B=@n1.A)))
  n13 = MaterialFunctionCall!(
   UV=NamedRerouteUsage()
   Segment Count=NamedRerouteUsage()
   Segment Spacing=ScalarParameter() {defaultValue=0.9, parameterName=SegmentSize}
   Fixed Spacing=Constant() {r=1}
  ) {materialFunction=/FortUILibrary/MaterialFunctions/Basics/MF_UI_RadialSegments}
  n14 = Constant() {r=0.5}
  n15 = Add(A=@n13.Segmented Lines Gradient, B=@n14) {constB=0.5}
  n16 = Add(A=@n14) {constB=0.1}
  NamedRerouteDeclaration(None=OneMinus(None=MaterialFunctionCall(DistanceField=@n15, Threshold=@n16) {materialFunction=&MF_UI_AntiAliasing}.Outer)) {name=SegmentMask, nodeColor=<rgba 0 1 0.6016 1>}
  NamedRerouteDeclaration(None=Floor(None=ScalarParameter() {defaultValue=8, parameterName=Segments})) {name=Segments, nodeColor=<rgba 1 0.3325 0.0183 1>}
  n17 = ConstantBiasScale(None=TextureCoordinate()) {bias=-0.5, scale=2}
  n18 = Constant()
  NamedRerouteDeclaration(None=Subtract(A=Distance(A=@n17, B=@n18), B=Constant() {r=0.5})) {name=SDFCircle, nodeColor=<rgba 1 0.1641 0 1>}
  n19 = NamedRerouteUsage()
  n20 = ScalarParameter() {defaultValue=-0.05, parameterName=RingInnerSize}
  n21 = ScalarParameter() {defaultValue=0.15, parameterName=HighlightThickness}
  n22 = Subtract(A=@n21, B=@n20)
  n23 = Multiply(A=Divide(A=Add(A=@n20, B=@n21), B=@n22), B=@n22)
  n24 = Subtract(A=Abs(None=Subtract(A=@n19, B=@n23)), B=@n22) {constB=0.05}
  NamedRerouteDeclaration(None=MaterialFunctionCall(DistanceField=Add(A=@n24) {constB=0.5}) {materialFunction=&MF_UI_AntiAliasing}.Inner) {name=SelectedRingMask, nodeColor=<rgba 0 0.4141 1 1>}
  n25 = NamedRerouteUsage()
  n26 = ScalarParameter() {defaultValue=-0.05, parameterName=RingInnerSize}
  n27 = ScalarParameter() {defaultValue=0.1, parameterName=RingThickness}
  n28 = Subtract(A=@n27, B=@n26)
  n29 = Multiply(A=Divide(A=Add(A=@n26, B=@n27), B=@n28), B=@n28)
  n30 = Subtract(A=Abs(None=Subtract(A=@n25, B=@n29)), B=@n28) {constB=0.05}
  NamedRerouteDeclaration(None=MaterialFunctionCall(DistanceField=Add(A=@n30) {constB=0.5}) {materialFunction=&MF_UI_AntiAliasing}.Inner) {name=UnselectedRingMask, nodeColor=<rgba 1 0.8672 0 1>}
  n31 = ScalarParameter() {parameterName=HighlightSegment}
  n32 = NamedRerouteUsage()
  n33 = Multiply(A=Divide(A=Floor(None=@n31), B=@n32)) {constB=-2}
  n34 = Divide(B=@n32)
  NamedRerouteDeclaration(None=Rotator(Time=Add(A=@n33, B=@n34)) {speed=3.1416}) {name=SegmentRotation, nodeColor=<rgba 0.7246 0 1 1>}
  n35 = Multiply(A=ScalarParameter() {defaultValue=0.1, parameterName=OuterGlow}) {constB=8}
  NamedRerouteDeclaration(None=OneMinus(None=SmoothStep(Max=@n35, Value=@n13.Segmented Lines Gradient))) {name=SegmentMaskGlow, nodeColor=<rgba 0.4297 0 1 1>}
  n36 = NamedRerouteUsage()
  n37 = ScalarParameter() {defaultValue=0.1, parameterName=OuterGlow}
  n38 = Divide(A=@n37) {constB=30}
  n39 = Add(A=@n20, B=@n38) {constB=0.02}
  n40 = Multiply(A=@n21) {constB=-1}
  n41 = Add(A=Add(A=@n21, B=Divide(A=@n40) {constB=10}) {constB=-0.01}, B=@n38) {constB=0.02}
  n42 = Subtract(A=@n41, B=@n39)
  n43 = Multiply(A=Divide(A=Add(A=@n39, B=@n41), B=@n42), B=@n42)
  n44 = Subtract(A=Abs(None=Subtract(A=@n36, B=@n43)), B=@n42) {constB=0.05}
  n45 = Add(A=@n37, B=Multiply(A=@n37, B=@n21))
  NamedRerouteDeclaration(None=OneMinus(None=SmoothStep(Max=@n45, Value=@n44) {constMax=0.1})) {name=SelectedGlowMask, nodeColor=<rgba 0.9453 0 1 1>}
  n46 = NamedRerouteUsage()
  n47 = ScalarParameter() {defaultValue=0.1, parameterName=OuterGlow}
  n48 = Divide(A=@n47) {constB=30}
  n49 = Add(A=@n26, B=@n48) {constB=0.02}
  n50 = Multiply(A=@n27) {constB=-1}
  n51 = Add(A=Add(A=@n27, B=Multiply(A=@n50) {constB=0.1}) {constB=-0.01}, B=@n48) {constB=0.02}
  n52 = Subtract(A=@n51, B=@n49)
  n53 = Multiply(A=Divide(A=Add(A=@n49, B=@n51), B=@n52), B=@n52)
  n54 = Subtract(A=Abs(None=Subtract(A=@n46, B=@n53)), B=@n52) {constB=0.05}
  n55 = Add(A=@n47, B=Multiply(A=@n47, B=@n27) {constA=2})
  NamedRerouteDeclaration(None=OneMinus(None=SmoothStep(Max=@n55, Value=@n54) {constMax=0.1})) {name=UnselectedGlowMask, nodeColor=<rgba 0.5469 0 1 1>}
  StaticBool() {value=true}
 ]
}
