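# M_UI_EmoteWheel_Slots is a UI material that procedurally draws the emote wheel's radial slots and center background.
# It converts rotated coordinates to radial space, divides them into subdivisions and shapes slots with box and circle SDF-style functions.
# Repeated GetSegmentByIndex subgraphs isolate indices 0 to 7 plus a parameter-driven index, providing per-slot and selection masks.
# Those masks, radial gradients and static switches drive layered lerp and max blends of color and opacity.
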Material /FortUILibrary/UIKit/Misc/Materials/M_UI_EmoteWheel_Slots
graph {
 expressions {
  MaterialExpressionAbs_0 {class=MaterialExpressionAbs, inputs [None {from=MaterialExpressionSubtract_0, output=""}], properties {}}
  MaterialExpressionAbs_1 {class=MaterialExpressionAbs, inputs [None {from=MaterialExpressionSubtract_6, output=""}], properties {}}
  MaterialExpressionAbs_2 {class=MaterialExpressionAbs, inputs [None {from=MaterialExpressionSubtract_2, output=""}], properties {}}
  MaterialExpressionAbs_3 {class=MaterialExpressionAbs, inputs [None {from=MaterialExpressionSubtract_3, output=""}], properties {}}
  MaterialExpressionAbs_4 {class=MaterialExpressionAbs, inputs [None {from=MaterialExpressionSubtract_4, output=""}], properties {}}
  MaterialExpressionAbs_5 {class=MaterialExpressionAbs, inputs [None {from=MaterialExpressionSubtract_5, output=""}], properties {}}
  MaterialExpressionAbs_6 {class=MaterialExpressionAbs, inputs [None {from=MaterialExpressionSubtract_7, output=""}], properties {}}
  MaterialExpressionAbs_7 {class=MaterialExpressionAbs, inputs [None {from=MaterialExpressionSubtract_8, output=""}], properties {}}
  MaterialExpressionAbs_8 {class=MaterialExpressionAbs, inputs [None {from=MaterialExpressionSubtract_9, output=""}], properties {}}
  MaterialExpressionAbs_9 {class=MaterialExpressionAbs, inputs [None {from=MaterialExpressionSubtract_10, output=""}], properties {}}
  MaterialExpressionAdd_0 {
   class=MaterialExpressionAdd
   inputs [
    A {from=MaterialExpressionNamedRerouteUsage_1}
    B {from=MaterialExpressionMultiply_20}
   ]
  }
  MaterialExpressionAdd_1 {
   class=MaterialExpressionAdd
   inputs [
    A {from=MaterialExpressionAdd_0}
    B {from=MaterialExpressionNamedRerouteUsage_5}
   ]
  }
  MaterialExpressionAdd_10 {
   class=MaterialExpressionAdd
   inputs [
    A {from=MaterialExpressionMultiply_4}
    B {from=MaterialExpressionNamedRerouteUsage_60}
   ]
   properties {constB=0.375}
  }
  MaterialExpressionAdd_11 {
   class=MaterialExpressionAdd
   inputs [
    A {from=MaterialExpressionAdd_10}
    B {from=MaterialExpressionMultiply_19}
   ]
   properties {constB=0.0375}
  }
  MaterialExpressionAdd_12 {
   class=MaterialExpressionAdd
   inputs [
    A {from=MaterialExpressionNamedRerouteUsage_25}
    B {from=MaterialExpressionMultiply_5}
   ]
  }
  MaterialExpressionAdd_13 {
   class=MaterialExpressionAdd
   inputs [
    A {from=MaterialExpressionFloor_4}
    B {from=""}
   ]
  }
  MaterialExpressionAdd_14 {
   class=MaterialExpressionAdd
   inputs [
    A {from=MaterialExpressionNamedRerouteUsage_77}
    B {from=""}
   ]
   properties {constB=0.1}
  }
  MaterialExpressionAdd_15 {
   class=MaterialExpressionAdd
   inputs [
    A {from=MaterialExpressionFloor_5}
    B {from=""}
   ]
  }
  MaterialExpressionAdd_16 {
   class=MaterialExpressionAdd
   inputs [
    A {from=MaterialExpressionFloor_6}
    B {from=""}
   ]
  }
  MaterialExpressionAdd_17 {
   class=MaterialExpressionAdd
   inputs [
    A {from=MaterialExpressionFloor_7}
    B {from=""}
   ]
  }
  MaterialExpressionAdd_18 {
   class=MaterialExpressionAdd
   inputs [
    A {from=MaterialExpressionFloor_8}
    B {from=""}
   ]
  }
  MaterialExpressionAdd_19 {
   class=MaterialExpressionAdd
   inputs [
    A {from=MaterialExpressionFloor_9}
    B {from=""}
   ]
  }
  MaterialExpressionAdd_2 {
   class=MaterialExpressionAdd
   inputs [
    A {from=MaterialExpressionFloor_2}
    B {from=""}
   ]
  }
  MaterialExpressionAdd_3 {
   class=MaterialExpressionAdd
   inputs [
    A {from=MaterialExpressionFloor_0}
    B {from=""}
   ]
  }
  MaterialExpressionAdd_4 {
   class=MaterialExpressionAdd
   inputs [
    A {from=MaterialExpressionNamedRerouteUsage_10}
    B {from=MaterialExpressionMultiply_18}
   ]
   properties {constB=0}
  }
  MaterialExpressionAdd_5 {
   class=MaterialExpressionAdd
   inputs [
    A {from=MaterialExpressionNamedRerouteUsage_53}
    B {from=MaterialExpressionMultiply_2}
   ]
  }
  MaterialExpressionAdd_6 {
   class=MaterialExpressionAdd
   inputs [
    A {from=MaterialExpressionAdd_8}
    B {from=MaterialExpressionNamedRerouteDeclaration_10}
   ]
   properties {constB=0}
  }
  MaterialExpressionAdd_7 {
   class=MaterialExpressionAdd
   inputs [
    A {from=MaterialExpressionFloor_3}
    B {from=""}
   ]
  }
  MaterialExpressionAdd_8 {
   class=MaterialExpressionAdd
   inputs [
    A {from=MaterialExpressionNamedRerouteUsage_43}
    B {from=MaterialExpressionMultiply_3}
   ]
  }
  MaterialExpressionAdd_9 {
   class=MaterialExpressionAdd
   inputs [
    A {from=MaterialExpressionNamedRerouteUsage_54}
    B {from=MaterialExpressionAdd_11}
   ]
  }
  MaterialExpressionAppendVector_0 {
   class=MaterialExpressionAppendVector
   inputs [
    A {from=MaterialExpressionNamedRerouteUsage_68, output=""}
    B {from=MaterialExpressionAdd_6, output=""}
   ]
  }
  MaterialExpressionAppendVector_11 {
   class=MaterialExpressionAppendVector
   inputs [
    A {from=MaterialExpressionNamedRerouteUsage_57, output=""}
    B {from=MaterialExpressionNamedRerouteDeclaration_18, output=""}
   ]
  }
  MaterialExpressionAppendVector_2 {
   class=MaterialExpressionAppendVector
   inputs [
    A {from=MaterialExpressionNamedRerouteUsage_64, output=""}
    B {from=MaterialExpressionSubtract_1, output=""}
   ]
  }
  MaterialExpressionAppendVector_6 {
   class=MaterialExpressionAppendVector
   inputs [
    A {from=MaterialExpressionNamedRerouteUsage_69, output=""}
    B {from=MaterialExpressionAdd_9, output=""}
   ]
  }
  MaterialExpressionAppendVector_7 {
   class=MaterialExpressionAppendVector
   inputs [
    A {from=MaterialExpressionOneMinus_4, output=""}
    B {from=MaterialExpressionConvert_1, output=""}
   ]
  }
  MaterialExpressionCeil_0 {class=MaterialExpressionCeil, inputs [None {from=MaterialExpressionMultiply_13, output=""}], properties {}}
  MaterialExpressionCeil_1 {class=MaterialExpressionCeil, inputs [None {from=MaterialExpressionReroute_7, output=SubdivisionAngle}], properties {}}
  MaterialExpressionCeil_10 {class=MaterialExpressionCeil, inputs [None {from=MaterialExpressionReroute_17, output=SubdivisionAngle}], properties {}}
  MaterialExpressionCeil_11 {class=MaterialExpressionCeil, inputs [None {from=MaterialExpressionReroute_18, output=SubdivisionAngle}], properties {}}
  MaterialExpressionCeil_2 {class=MaterialExpressionCeil, inputs [None {from=MaterialExpressionNamedRerouteUsage_8, output=""}], properties {}}
  MaterialExpressionCeil_3 {class=MaterialExpressionCeil, inputs [None {from=MaterialExpressionReroute_12, output=SubdivisionAngle}], properties {}}
  MaterialExpressionCeil_4 {class=MaterialExpressionCeil, inputs [None {from=MaterialExpressionReroute_9, output=SubdivisionAngle}], properties {}}
  MaterialExpressionCeil_5 {class=MaterialExpressionCeil, inputs [None {from=MaterialExpressionReroute_15, output=SubdivisionAngle}], properties {}}
  MaterialExpressionCeil_6 {class=MaterialExpressionCeil, inputs [None {from=MaterialExpressionSaturate_2, output=""}], properties {}}
  MaterialExpressionCeil_7 {class=MaterialExpressionCeil, inputs [None {from=MaterialExpressionReroute_7, output=SubdivisionAngle}], properties {}}
  MaterialExpressionCeil_8 {class=MaterialExpressionCeil, inputs [None {from=MaterialExpressionReroute_13, output=SubdivisionAngle}], properties {}}
  MaterialExpressionCeil_9 {class=MaterialExpressionCeil, inputs [None {from=MaterialExpressionReroute_9, output=SubdivisionAngle}], properties {}}
  MaterialExpressionComposite_0 {
   class=MaterialExpressionComposite
   inputs [
    SubdivisionAngle {from=MaterialExpressionNamedRerouteUsage_11, output=""}
    Index {from=MaterialExpressionScalarParameter_3, output=""}
   ]
   properties {subgraphName=GetSegmentByIndex}
  }
  MaterialExpressionComposite_1 {
   class=MaterialExpressionComposite
   inputs [
    SubdivisionAngle {from=MaterialExpressionNamedRerouteUsage_4, output=""}
    Index {from=MaterialExpressionConstant_4, output=""}
   ]
   properties {subgraphName=GetSegmentByIndex}
  }
  MaterialExpressionComposite_2 {
   class=MaterialExpressionComposite
   inputs [
    SubdivisionAngle {from=MaterialExpressionNamedRerouteUsage_18, output=""}
    Index {from=MaterialExpressionConstant_5, output=""}
   ]
   properties {subgraphName=GetSegmentByIndex}
  }
  MaterialExpressionComposite_3 {
   class=MaterialExpressionComposite
   inputs [
    SubdivisionAngle {from=MaterialExpressionNamedRerouteUsage_20, output=""}
    Index {from=MaterialExpressionConstant_6, output=""}
   ]
   properties {subgraphName=GetSegmentByIndex}
  }
  MaterialExpressionComposite_4 {
   class=MaterialExpressionComposite
   inputs [
    SubdivisionAngle {from=MaterialExpressionNamedRerouteUsage_27, output=""}
    Index {from=MaterialExpressionConstant_8, output=""}
   ]
   properties {subgraphName=GetSegmentByIndex}
  }
  MaterialExpressionComposite_5 {
   class=MaterialExpressionComposite
   inputs [
    SubdivisionAngle {from=MaterialExpressionNamedRerouteUsage_37, output=""}
    Index {from=MaterialExpressionConstant_9, output=""}
   ]
   properties {subgraphName=GetSegmentByIndex}
  }
  MaterialExpressionComposite_6 {
   class=MaterialExpressionComposite
   inputs [
    SubdivisionAngle {from=MaterialExpressionNamedRerouteUsage_34, output=""}
    Index {from=MaterialExpressionConstant_10, output=""}
   ]
   properties {subgraphName=GetSegmentByIndex}
  }
  MaterialExpressionComposite_7 {
   class=MaterialExpressionComposite
   inputs [
    SubdivisionAngle {from=MaterialExpressionNamedRerouteUsage_32, output=""}
    Index {from=MaterialExpressionConstant_11, output=""}
   ]
   properties {subgraphName=GetSegmentByIndex}
  }
  MaterialExpressionComposite_8 {
   class=MaterialExpressionComposite
   inputs [
    SubdivisionAngle {from=MaterialExpressionNamedRerouteUsage_30, output=""}
    Index {from=MaterialExpressionConstant_14, output=""}
   ]
   properties {subgraphName=GetSegmentByIndex}
  }
  MaterialExpressionConstant2Vector_0 {class=MaterialExpressionConstant2Vector, properties {g=1}}
  MaterialExpressionConstant2Vector_1 {class=MaterialExpressionConstant2Vector, properties {g=1, r=0.03}}
  MaterialExpressionConstant2Vector_5 {class=MaterialExpressionConstant2Vector, properties {g=1, r=1}}
  MaterialExpressionConstant3Vector_2 {class=MaterialExpressionConstant3Vector, properties {constant=<rgba 1 1 1 1>}}
  MaterialExpressionConstant3Vector_3 {class=MaterialExpressionConstant3Vector, properties {constant=<rgba 1 1 1 1>}}
  MaterialExpressionConstant4Vector_0 {class=MaterialExpressionConstant4Vector, inputs [], properties {constant=<rgba 1 1 1 1>}}
  MaterialExpressionConstant_0 {class=MaterialExpressionConstant, properties {r=1}}
  MaterialExpressionConstant_1 {class=MaterialExpressionConstant, properties {r=-1}}
  MaterialExpressionConstant_10 {class=MaterialExpressionConstant, properties {r=6}}
  MaterialExpressionConstant_11 {class=MaterialExpressionConstant, properties {r=5}}
  MaterialExpressionConstant_12 {class=MaterialExpressionConstant, properties {r=0.4}}
  MaterialExpressionConstant_13 {class=MaterialExpressionConstant, properties {r=0.4}}
  MaterialExpressionConstant_14 {class=MaterialExpressionConstant, properties {r=4}}
  MaterialExpressionConstant_15 {class=MaterialExpressionConstant, properties {r=7}}
  MaterialExpressionConstant_16 {class=MaterialExpressionConstant, properties {r=15}}
  MaterialExpressionConstant_2 {class=MaterialExpressionConstant, properties {r=1}}
  MaterialExpressionConstant_3 {class=MaterialExpressionConstant, properties {r=-1}}
  MaterialExpressionConstant_4 {class=MaterialExpressionConstant}
  MaterialExpressionConstant_5 {class=MaterialExpressionConstant, properties {r=1}}
  MaterialExpressionConstant_6 {class=MaterialExpressionConstant, properties {r=2}}
  MaterialExpressionConstant_7 {class=MaterialExpressionConstant, properties {r=0.01}}
  MaterialExpressionConstant_8 {class=MaterialExpressionConstant, properties {r=3}}
  MaterialExpressionConstant_9 {class=MaterialExpressionConstant, properties {r=7}}
  MaterialExpressionConvert_0 {
   class=MaterialExpressionConvert
   inputs [None {from=MaterialExpressionMaterialFunctionCall_2, output=Radial Coordinates}]
   properties {
    nodeName=Break Float2
    convertInputs [{
      type=Vector2
      expressionInput {
       expression=&self:MaterialExpressionMaterialFunctionCall_2
       inputName=None
       mask=0
       maskA=0
       maskB=0
       maskG=0
       maskR=0
       outputIndex=0
      }
     }]
    convertOutputs [{type=Scalar}, {type=Scalar}]
   }
  }
  MaterialExpressionConvert_1 {
   class=MaterialExpressionConvert
   inputs [None {from=MaterialExpressionNamedRerouteDeclaration_3}]
   properties {
    nodeName=Break Float2
    convertInputs [{
      type=Vector2
      expressionInput {
       expression=&self:MaterialExpressionNamedRerouteDeclaration_3
       inputName=None
       mask=0
       maskA=0
       maskB=0
       maskG=0
       maskR=0
       outputIndex=0
      }
     }]
    convertOutputs [{type=Scalar}, {type=Scalar}]
   }
  }
  MaterialExpressionConvert_3 {
   class=MaterialExpressionConvert
   inputs [None {from=MaterialExpressionNamedRerouteUsage_70}]
   properties {
    convertInputs [{
      type=Vector4
      defaultValue=<rgba 0 0 0 1>
      expressionInput {
       expression=&self:MaterialExpressionNamedRerouteUsage_70
       inputName=None
       mask=0
       maskA=0
       maskB=0
       maskG=0
       maskR=0
       outputIndex=0
      }
     }]
    convertOutputs [{type=Vector3}]
   }
  }
  MaterialExpressionConvert_5 {
   class=MaterialExpressionConvert
   inputs [
    None {from=MaterialExpressionDivide_1}
    None {from=""}
   ]
   properties {
    nodeName=Make Vector2
    convertInputs [
     {
      type=Scalar
      expressionInput {
       expression=&self:MaterialExpressionDivide_1
       inputName=None
       mask=0
       maskA=0
       maskB=0
       maskG=0
       maskR=0
       outputIndex=0
      }
     }
     {
      type=Scalar
      defaultValue=<rgba 1 0 0 1>
      expressionInput {expression=null, inputName=None, mask=0, maskA=0, maskB=0, maskG=0, maskR=0, outputIndex=0}
     }
    ]
    convertOutputs [{type=Vector2}]
   }
  }
  MaterialExpressionDesaturation_0 {
   class=MaterialExpressionDesaturation
   inputs [
    None {from=MaterialExpressionNamedRerouteDeclaration_63, output=""}
    Fraction {from="", output=""}
   ]
   properties {luminanceFactors=<rgba 1 1 1 1>}
  }
  MaterialExpressionDivide_0 {
   class=MaterialExpressionDivide
   inputs [
    A {from=MaterialExpressionMultiply_0}
    B {from=MaterialExpressionNamedRerouteDeclaration_33}
   ]
  }
  MaterialExpressionDivide_1 {
   class=MaterialExpressionDivide
   inputs [
    A {from=MaterialExpressionSubtract_13}
    B {from=MaterialExpressionConstant_15}
   ]
  }
  MaterialExpressionDivide_2 {
   class=MaterialExpressionDivide
   inputs [
    A {from=""}
    B {from=MaterialExpressionMax_1}
   ]
   properties {constA=-1}
  }
  MaterialExpressionDotProduct_0 {
   class=MaterialExpressionDotProduct
   inputs [
    A {from=MaterialExpressionNamedRerouteUsage_38, output=""}
    B {from=MaterialExpressionNamedRerouteDeclaration_21, output=""}
   ]
   properties {}
  }
  MaterialExpressionFloor_0 {class=MaterialExpressionFloor, inputs [None {from=MaterialExpressionReroute_8, output=Index}], properties {}}
  MaterialExpressionFloor_1 {class=MaterialExpressionFloor, inputs [None {from=MaterialExpressionScalarParameter_0, output=""}], properties {}}
  MaterialExpressionFloor_2 {class=MaterialExpressionFloor, inputs [None {from=MaterialExpressionReroute_13, output=Index}], properties {}}
  MaterialExpressionFloor_3 {class=MaterialExpressionFloor, inputs [None {from=MaterialExpressionReroute_10, output=Index}], properties {}}
  MaterialExpressionFloor_4 {class=MaterialExpressionFloor, inputs [None {from=MaterialExpressionReroute_16, output=Index}], properties {}}
  MaterialExpressionFloor_5 {class=MaterialExpressionFloor, inputs [None {from=MaterialExpressionReroute_8, output=Index}], properties {}}
  MaterialExpressionFloor_6 {class=MaterialExpressionFloor, inputs [None {from=MaterialExpressionReroute_6, output=Index}], properties {}}
  MaterialExpressionFloor_7 {class=MaterialExpressionFloor, inputs [None {from=MaterialExpressionReroute_15, output=Index}], properties {}}
  MaterialExpressionFloor_8 {class=MaterialExpressionFloor, inputs [None {from=MaterialExpressionReroute_7, output=Index}], properties {}}
  MaterialExpressionFloor_9 {class=MaterialExpressionFloor, inputs [None {from=MaterialExpressionReroute_10, output=Index}], properties {}}
  MaterialExpressionFrac_3 {class=MaterialExpressionFrac, inputs [None {from=MaterialExpressionNamedRerouteDeclaration_9, output=""}], properties {}}
  MaterialExpressionLinearInterpolate_0 {
   class=MaterialExpressionLinearInterpolate
   inputs [
    A {from=MaterialExpressionLinearInterpolate_10}
    B {from=MaterialExpressionStaticSwitchParameter_1}
    Alpha {from=MaterialExpressionCeil_0}
   ]
   properties {constAlpha=0, constB=0.025}
  }
  MaterialExpressionLinearInterpolate_1 {
   class=MaterialExpressionLinearInterpolate
   inputs [
    A {from=MaterialExpressionNamedRerouteUsage_35}
    B {from=MaterialExpressionMax_3}
    Alpha {from=MaterialExpressionReroute_3}
   ]
  }
  MaterialExpressionLinearInterpolate_10 {
   class=MaterialExpressionLinearInterpolate
   inputs [
    A {from=MaterialExpressionLinearInterpolate_13}
    B {from=MaterialExpressionVectorParameter_3, output=RGB}
    Alpha {from=MaterialExpressionNamedRerouteUsage_33}
   ]
  }
  MaterialExpressionLinearInterpolate_11 {
   class=MaterialExpressionLinearInterpolate
   inputs [
    A {from=MaterialExpressionNamedRerouteDeclaration_8}
    B {from=MaterialExpressionNamedRerouteUsage_42}
    Alpha {from=MaterialExpressionMultiply_39}
   ]
  }
  MaterialExpressionLinearInterpolate_12 {
   class=MaterialExpressionLinearInterpolate
   inputs [
    A {from=MaterialExpressionVectorParameter_0, output=RGB}
    B {from=MaterialExpressionNamedRerouteDeclaration_63}
    Alpha {from=MaterialExpressionMultiply_31}
   ]
  }
  MaterialExpressionLinearInterpolate_13 {
   class=MaterialExpressionLinearInterpolate
   inputs [
    A {from=MaterialExpressionStaticSwitchParameter_5}
    B {from=MaterialExpressionNamedRerouteDeclaration_31}
    Alpha {from=MaterialExpressionNamedRerouteUsage_63}
   ]
  }
  MaterialExpressionLinearInterpolate_14 {
   class=MaterialExpressionLinearInterpolate
   inputs [
    A {from=MaterialExpressionMultiply_34}
    B {from=""}
    Alpha {from=MaterialExpressionNamedRerouteDeclaration_17}
   ]
  }
  MaterialExpressionLinearInterpolate_15 {
   class=MaterialExpressionLinearInterpolate
   inputs [
    A {from=""}
    B {from=MaterialExpressionSaturate_13}
    Alpha {from=MaterialExpressionScalarParameter_13}
   ]
  }
  MaterialExpressionLinearInterpolate_16 {
   class=MaterialExpressionLinearInterpolate
   inputs [
    A {from=MaterialExpressionReroute_22}
    B {from=MaterialExpressionReroute_1}
    Alpha {from=MaterialExpressionMultiply_38}
   ]
  }
  MaterialExpressionLinearInterpolate_17 {
   class=MaterialExpressionLinearInterpolate
   inputs [
    A {from=MaterialExpressionNamedRerouteUsage_48}
    B {from=""}
    Alpha {from=MaterialExpressionOneMinus_17}
   ]
   properties {constB=0}
  }
  MaterialExpressionLinearInterpolate_18 {
   class=MaterialExpressionLinearInterpolate
   inputs [
    A {from=MaterialExpressionNamedRerouteUsage_47}
    B {from=""}
    Alpha {from=MaterialExpressionOneMinus_19}
   ]
   properties {constB=0}
  }
  MaterialExpressionLinearInterpolate_19 {
   class=MaterialExpressionLinearInterpolate
   inputs [
    A {from=MaterialExpressionNamedRerouteUsage_46}
    B {from=""}
    Alpha {from=MaterialExpressionOneMinus_20}
   ]
   properties {constB=0}
  }
  MaterialExpressionLinearInterpolate_2 {
   class=MaterialExpressionLinearInterpolate
   inputs [
    A {from=MaterialExpressionConstant3Vector_3}
    B {from=MaterialExpressionConstant3Vector_2}
    Alpha {from=MaterialExpressionNamedRerouteUsage_49}
   ]
  }
  MaterialExpressionLinearInterpolate_20 {
   class=MaterialExpressionLinearInterpolate
   inputs [
    A {from=MaterialExpressionNamedRerouteUsage_45}
    B {from=""}
    Alpha {from=MaterialExpressionOneMinus_21}
   ]
   properties {constB=0}
  }
  MaterialExpressionLinearInterpolate_21 {
   class=MaterialExpressionLinearInterpolate
   inputs [
    A {from=MaterialExpressionNamedRerouteUsage_41}
    B {from=""}
    Alpha {from=MaterialExpressionOneMinus_22}
   ]
   properties {constB=0}
  }
  MaterialExpressionLinearInterpolate_22 {
   class=MaterialExpressionLinearInterpolate
   inputs [
    A {from=MaterialExpressionReroute_19}
    B {from=MaterialExpressionMax_15}
    Alpha {from=MaterialExpressionNamedRerouteUsage_83}
   ]
  }
  MaterialExpressionLinearInterpolate_23 {
   class=MaterialExpressionLinearInterpolate
   inputs [
    A {from=MaterialExpressionLinearInterpolate_11}
    B {from=MaterialExpressionLinearInterpolate_16}
    Alpha {from=MaterialExpressionNamedRerouteUsage_84}
   ]
  }
  MaterialExpressionLinearInterpolate_24 {
   class=MaterialExpressionLinearInterpolate
   inputs [
    A {from=MaterialExpressionNamedRerouteUsage_40}
    B {from=""}
    Alpha {from=MaterialExpressionOneMinus_25}
   ]
   properties {constB=0}
  }
  MaterialExpressionLinearInterpolate_25 {
   class=MaterialExpressionLinearInterpolate
   inputs [
    A {from=MaterialExpressionNamedRerouteUsage_39}
    B {from=""}
    Alpha {from=MaterialExpressionOneMinus_27}
   ]
   properties {constB=0}
  }
  MaterialExpressionLinearInterpolate_3 {
   class=MaterialExpressionLinearInterpolate
   inputs [
    A {from=MaterialExpressionReroute_5}
    B {from=MaterialExpressionNamedRerouteUsage_31}
    Alpha {from=MaterialExpressionNamedRerouteUsage_13}
   ]
  }
  MaterialExpressionLinearInterpolate_4 {
   class=MaterialExpressionLinearInterpolate
   inputs [
    A {from=MaterialExpressionNamedRerouteUsage_50}
    B {from=""}
    Alpha {from=MaterialExpressionOneMinus_18}
   ]
   properties {constB=0}
  }
  MaterialExpressionLinearInterpolate_5 {
   class=MaterialExpressionLinearInterpolate
   inputs [
    A {from=MaterialExpressionVectorParameter_1, output=RGB}
    B {from=MaterialExpressionVectorParameter_2, output=RGB}
    Alpha {from=MaterialExpressionMaterialFunctionCall_6, output=Result}
   ]
   properties {constA=1, constB=0.319}
  }
  MaterialExpressionLinearInterpolate_6 {
   class=MaterialExpressionLinearInterpolate
   inputs [
    A {from=MaterialExpressionLinearInterpolate_1}
    B {from=MaterialExpressionReroute_2}
    Alpha {from=MaterialExpressionNamedRerouteUsage_23}
   ]
   properties {constA=0.12}
  }
  MaterialExpressionLinearInterpolate_7 {
   class=MaterialExpressionLinearInterpolate
   inputs [
    A {from=MaterialExpressionVectorParameter_1, output=A}
    B {from=MaterialExpressionVectorParameter_2, output=A}
    Alpha {from=MaterialExpressionMaterialFunctionCall_6, output=Result}
   ]
   properties {constA=0.7, constB=0.3}
  }
  MaterialExpressionLinearInterpolate_8 {
   class=MaterialExpressionLinearInterpolate
   inputs [
    A {from=MaterialExpressionMultiply_21}
    B {from=""}
    Alpha {from=MaterialExpressionNamedRerouteDeclaration_27}
   ]
  }
  MaterialExpressionLinearInterpolate_9 {
   class=MaterialExpressionLinearInterpolate
   inputs [
    A {from=MaterialExpressionLinearInterpolate_22}
    B {from=MaterialExpressionLinearInterpolate_23}
    Alpha {from=MaterialExpressionNamedRerouteUsage_82}
   ]
  }
  MaterialExpressionMaterialFunctionCall_0 {
   class=MaterialExpressionMaterialFunctionCall
   inputs [
    UVs {from=MaterialExpressionNamedRerouteUsage_0}
    Size {from=MaterialExpressionSaturate_0}
    OutlineThickness {from=""}
    OutlineAlignment {from=""}
    GlowMin {from=""}
    GlowMax {from=""}
   ]
   properties {materialFunction=&MF_UI_Circle}
  }
  MaterialExpressionMaterialFunctionCall_1 {
   class=MaterialExpressionMaterialFunctionCall
   inputs [
    UVs {from=MaterialExpressionMultiply_47}
    Size {from=MaterialExpressionAppendVector_2}
    CornerRoundness {from=MaterialExpressionMultiply_1}
    OutlineThickness {from=MaterialExpressionScalarParameter_2}
    OutlineAlignment {from=MaterialExpressionConstant_3}
    GlowMin {from=""}
    GlowMax {from=""}
    KeepAspectRatio {from=MaterialExpressionStaticBool_0}
    PixelSize {from=""}
   ]
   properties {materialFunction=&MF_UI_Box}
  }
  MaterialExpressionMaterialFunctionCall_2 {
   class=MaterialExpressionMaterialFunctionCall
   inputs [Vector or UVs {from=MaterialExpressionRotator_0}]
   properties {materialFunction=/FortUILibrary/MaterialFunctions/Basics/MF_UI_CheaperVectorToRadial}
  }
  MaterialExpressionMaterialFunctionCall_23 {
   class=MaterialExpressionMaterialFunctionCall
   inputs [
    UVs {from=MaterialExpressionAppendVector_6}
    Size {from=MaterialExpressionAppendVector_11}
    CornerRoundness {from=MaterialExpressionMultiply_56}
    OutlineThickness {from=""}
    OutlineAlignment {from=""}
    GlowMin {from=""}
    GlowMax {from=""}
    KeepAspectRatio {from=MaterialExpressionStaticBool_1}
    PixelSize {from=""}
   ]
   properties {materialFunction=&MF_UI_Box}
  }
  MaterialExpressionMaterialFunctionCall_3 {class=MaterialExpressionMaterialFunctionCall, inputs [], properties {materialFunction=&GetUserInterfaceUV}}
  MaterialExpressionMaterialFunctionCall_4 {
   class=MaterialExpressionMaterialFunctionCall
   inputs [
    UV {from=MaterialExpressionMaterialFunctionCall_9, output=UVs}
    GradientStart {from=MaterialExpressionConstant2Vector_0, output=R}
    GradientEnd {from=MaterialExpressionConstant2Vector_0, output=R}
   ]
   properties {materialFunction=&MF_UI_RadialGradient}
  }
  MaterialExpressionMaterialFunctionCall_5 {
   class=MaterialExpressionMaterialFunctionCall
   inputs [
    UVs {from=MaterialExpressionNamedRerouteUsage_9}
    Size {from=MaterialExpressionConstant_0}
    OutlineThickness {from=MaterialExpressionSaturate_1}
    OutlineAlignment {from=MaterialExpressionConstant_1}
    GlowMin {from=""}
    GlowMax {from=""}
   ]
   properties {materialFunction=&MF_UI_Circle}
  }
  MaterialExpressionMaterialFunctionCall_6 {
   class=MaterialExpressionMaterialFunctionCall
   inputs [
    Input {from=MaterialExpressionSaturate_16}
    Input Low {from=MaterialExpressionOneMinus_1}
    Input High {from=MaterialExpressionConstant_2}
   ]
   properties {
    desc=Adjust the gradient range based on main subdivisions range
    materialFunction=/Engine/Functions/Engine_MaterialFunctions03/Math/RemapValueRangeNormalized
   }
  }
  MaterialExpressionMaterialFunctionCall_7 {
   class=MaterialExpressionMaterialFunctionCall
   inputs ["UVs (Optional)" {from=""}]
   properties {materialFunction=/Engine/Functions/UserInterface/GetSlatePost0}
  }
  MaterialExpressionMaterialFunctionCall_8 {
   class=MaterialExpressionMaterialFunctionCall
   inputs [
    UV {from=""}
    GradientStart {from=MaterialExpressionConstant2Vector_1, output=R}
    GradientEnd {from=MaterialExpressionConstant2Vector_1, output=R}
   ]
   properties {materialFunction=&MF_UI_RadialGradient}
  }
  MaterialExpressionMaterialFunctionCall_9 {
   class=MaterialExpressionMaterialFunctionCall
   inputs [
    UVs {from=MaterialExpressionTextureCoordinate_0}
    Texture Scale {from=MaterialExpressionScalarParameter_22}
   ]
   properties {materialFunction=&ScaleUVsByCenter}
  }
  MaterialExpressionMax_0 {
   class=MaterialExpressionMax
   inputs [
    A {from=MaterialExpressionLinearInterpolate_8}
    B {from=MaterialExpressionLinearInterpolate_14}
   ]
  }
  MaterialExpressionMax_1 {
   class=MaterialExpressionMax
   inputs [
    A {from=MaterialExpressionNamedRerouteUsage_6}
    B {from=""}
   ]
   properties {constB=0}
  }
  MaterialExpressionMax_10 {
   class=MaterialExpressionMax
   inputs [
    A {from=MaterialExpressionMax_9}
    B {from=MaterialExpressionLinearInterpolate_19}
   ]
  }
  MaterialExpressionMax_11 {
   class=MaterialExpressionMax
   inputs [
    A {from=MaterialExpressionMax_10}
    B {from=MaterialExpressionLinearInterpolate_18}
   ]
  }
  MaterialExpressionMax_12 {
   class=MaterialExpressionMax
   inputs [
    A {from=MaterialExpressionMax_11}
    B {from=MaterialExpressionLinearInterpolate_17}
   ]
  }
  MaterialExpressionMax_13 {
   class=MaterialExpressionMax
   inputs [
    A {from=MaterialExpressionMax_12}
    B {from=MaterialExpressionLinearInterpolate_4}
   ]
  }
  MaterialExpressionMax_15 {
   class=MaterialExpressionMax
   inputs [
    A {from=MaterialExpressionReroute_20}
    B {from=MaterialExpressionNamedRerouteUsage_86}
   ]
  }
  MaterialExpressionMax_2 {
   class=MaterialExpressionMax
   inputs [
    A {from=MaterialExpressionMultiply_13}
    B {from=MaterialExpressionStaticSwitchParameter_6}
   ]
  }
  MaterialExpressionMax_3 {
   class=MaterialExpressionMax
   inputs [
    A {from=MaterialExpressionReroute_4}
    B {from=MaterialExpressionLinearInterpolate_3}
   ]
  }
  MaterialExpressionMax_6 {
   class=MaterialExpressionMax
   inputs [
    A {from=MaterialExpressionNamedRerouteUsage_76}
    B {from=MaterialExpressionNamedRerouteUsage_75}
   ]
  }
  MaterialExpressionMax_7 {
   class=MaterialExpressionMax
   inputs [
    A {from=MaterialExpressionLinearInterpolate_25}
    B {from=MaterialExpressionLinearInterpolate_24}
   ]
   properties {constB=0}
  }
  MaterialExpressionMax_8 {
   class=MaterialExpressionMax
   inputs [
    A {from=MaterialExpressionMax_7}
    B {from=MaterialExpressionLinearInterpolate_21}
   ]
  }
  MaterialExpressionMax_9 {
   class=MaterialExpressionMax
   inputs [
    A {from=MaterialExpressionMax_8}
    B {from=MaterialExpressionLinearInterpolate_20}
   ]
  }
  MaterialExpressionMin_0 {
   class=MaterialExpressionMin
   inputs [
    A {from=MaterialExpressionFrac_3, output=""}
    B {from=MaterialExpressionOneMinus_12, output=""}
   ]
   properties {}
  }
  MaterialExpressionMultiply_0 {
   class=MaterialExpressionMultiply
   inputs [
    A {from=MaterialExpressionNamedRerouteDeclaration_4}
    B {from=MaterialExpressionConvert_0}
   ]
   properties {constB=8}
  }
  MaterialExpressionMultiply_1 {
   class=MaterialExpressionMultiply
   inputs [
    A {from=MaterialExpressionSubtract_1}
    B {from=MaterialExpressionScalarParameter_7}
   ]
  }
  MaterialExpressionMultiply_10 {
   class=MaterialExpressionMultiply
   inputs [
    A {from=MaterialExpressionMultiply_60}
    B {from=MaterialExpressionMultiply_62}
   ]
  }
  MaterialExpressionMultiply_11 {
   class=MaterialExpressionMultiply
   inputs [
    A {from=MaterialExpressionAdd_4}
    B {from=""}
   ]
   properties {constB=0.5}
  }
  MaterialExpressionMultiply_12 {
   class=MaterialExpressionMultiply
   inputs [
    A {from=MaterialExpressionOneMinus_0}
    B {from=MaterialExpressionScalarParameter_6}
   ]
   properties {constB=0.5}
  }
  MaterialExpressionMultiply_13 {
   class=MaterialExpressionMultiply
   inputs [
    A {from=MaterialExpressionNamedRerouteUsage_7}
    B {from=MaterialExpressionStaticSwitchParameter_2}
   ]
   properties {constB=0.1}
  }
  MaterialExpressionMultiply_14 {
   class=MaterialExpressionMultiply
   inputs [
    A {from=MaterialExpressionNamedRerouteDeclaration_41}
    B {from=MaterialExpressionNamedRerouteDeclaration_40}
   ]
  }
  MaterialExpressionMultiply_15 {
   class=MaterialExpressionMultiply
   inputs [
    A {from=MaterialExpressionNamedRerouteDeclaration_42}
    B {from=MaterialExpressionNamedRerouteDeclaration_48}
   ]
  }
  MaterialExpressionMultiply_16 {
   class=MaterialExpressionMultiply
   inputs [
    A {from=MaterialExpressionNamedRerouteDeclaration_43}
    B {from=MaterialExpressionNamedRerouteDeclaration_49}
   ]
  }
  MaterialExpressionMultiply_17 {
   class=MaterialExpressionMultiply
   inputs [
    A {from=MaterialExpressionNamedRerouteDeclaration_44}
    B {from=MaterialExpressionNamedRerouteDeclaration_50}
   ]
  }
  MaterialExpressionMultiply_18 {
   class=MaterialExpressionMultiply
   inputs [
    A {from=MaterialExpressionNamedRerouteUsage_19}
    B {from=""}
   ]
   properties {constB=2}
  }
  MaterialExpressionMultiply_19 {
   class=MaterialExpressionMultiply
   inputs [
    A {from=MaterialExpressionNamedRerouteUsage_21}
    B {from=""}
   ]
   properties {constB=2}
  }
  MaterialExpressionMultiply_2 {
   class=MaterialExpressionMultiply
   inputs [
    A {from=MaterialExpressionConstant4Vector_0, output=RGB}
    B {from=""}
   ]
   properties {constB=0.04}
  }
  MaterialExpressionMultiply_20 {
   class=MaterialExpressionMultiply
   inputs [
    A {from=MaterialExpressionNamedRerouteUsage_24}
    B {from=""}
   ]
   properties {constB=2}
  }
  MaterialExpressionMultiply_21 {
   class=MaterialExpressionMultiply
   inputs [
    A {from=MaterialExpressionShadingPathSwitch_4}
    B {from=MaterialExpressionNamedRerouteUsage_26}
   ]
  }
  MaterialExpressionMultiply_22 {
   class=MaterialExpressionMultiply
   inputs [
    A {from=MaterialExpressionAdd_5}
    B {from=MaterialExpressionVectorParameter_5, output=RGB}
   ]
  }
  MaterialExpressionMultiply_23 {
   class=MaterialExpressionMultiply
   inputs [
    A {from=MaterialExpressionNamedRerouteUsage_55}
    B {from=MaterialExpressionVectorParameter_4, output=RGB}
   ]
  }
  MaterialExpressionMultiply_24 {
   class=MaterialExpressionMultiply
   inputs [
    A {from=MaterialExpressionNamedRerouteDeclaration_45}
    B {from=MaterialExpressionNamedRerouteDeclaration_51}
   ]
  }
  MaterialExpressionMultiply_25 {
   class=MaterialExpressionMultiply
   inputs [
    A {from=MaterialExpressionNamedRerouteDeclaration_46}
    B {from=MaterialExpressionNamedRerouteDeclaration_52}
   ]
  }
  MaterialExpressionMultiply_26 {
   class=MaterialExpressionMultiply
   inputs [
    A {from=MaterialExpressionNamedRerouteDeclaration_47}
    B {from=MaterialExpressionNamedRerouteDeclaration_53}
   ]
  }
  MaterialExpressionMultiply_27 {
   class=MaterialExpressionMultiply
   inputs [
    A {from=MaterialExpressionLinearInterpolate_6}
    B {from=MaterialExpressionNamedRerouteUsage_52}
   ]
   properties {constB=0.4}
  }
  MaterialExpressionMultiply_29 {
   class=MaterialExpressionMultiply
   inputs [
    A {from=MaterialExpressionNamedRerouteUsage_23}
    B {from=MaterialExpressionNamedRerouteUsage_28}
   ]
  }
  MaterialExpressionMultiply_3 {
   class=MaterialExpressionMultiply
   inputs [
    A {from=MaterialExpressionOneMinus_23}
    B {from=""}
   ]
   properties {constB=-0.5, desc=Set Pivot Point Offset}
  }
  MaterialExpressionMultiply_31 {
   class=MaterialExpressionMultiply
   inputs [
    A {from=MaterialExpressionNamedRerouteDeclaration_65}
    B {from=MaterialExpressionLinearInterpolate_15}
   ]
  }
  MaterialExpressionMultiply_32 {
   class=MaterialExpressionMultiply
   inputs [
    A {from=MaterialExpressionMaterialFunctionCall_4, output=Smooth}
    B {from=""}
   ]
  }
  MaterialExpressionMultiply_33 {
   class=MaterialExpressionMultiply
   inputs [
    A {from=MaterialExpressionNamedRerouteUsage_78}
    B {from=MaterialExpressionSaturate_14}
   ]
  }
  MaterialExpressionMultiply_34 {
   class=MaterialExpressionMultiply
   inputs [
    A {from=MaterialExpressionNamedRerouteUsage_71}
    B {from=MaterialExpressionNamedRerouteUsage_67}
   ]
  }
  MaterialExpressionMultiply_35 {
   class=MaterialExpressionMultiply
   inputs [
    A {from=MaterialExpressionSaturate_17}
    B {from=MaterialExpressionNamedRerouteUsage_81}
   ]
  }
  MaterialExpressionMultiply_36 {
   class=MaterialExpressionMultiply
   inputs [
    A {from=MaterialExpressionMultiply_33}
    B {from=""}
   ]
   properties {constB=3}
  }
  MaterialExpressionMultiply_37 {
   class=MaterialExpressionMultiply
   inputs [
    A {from=MaterialExpressionMaterialFunctionCall_8, output=Smooth}
    B {from=""}
   ]
  }
  MaterialExpressionMultiply_38 {
   class=MaterialExpressionMultiply
   inputs [
    A {from=MaterialExpressionNamedRerouteUsage_85}
    B {from=MaterialExpressionNamedRerouteUsage_15}
   ]
  }
  MaterialExpressionMultiply_39 {
   class=MaterialExpressionMultiply
   inputs [
    A {from=MaterialExpressionNamedRerouteUsage_44}
    B {from=MaterialExpressionOneMinus_16}
   ]
  }
  MaterialExpressionMultiply_4 {
   class=MaterialExpressionMultiply
   inputs [
    A {from=MaterialExpressionOneMinus_24}
    B {from=""}
   ]
   properties {constB=-0.5, desc=Set Pivot Point Offset}
  }
  MaterialExpressionMultiply_40 {
   class=MaterialExpressionMultiply
   inputs [
    A {from=MaterialExpressionOneMinus_8}
    B {from=MaterialExpressionOneMinus_7}
   ]
  }
  MaterialExpressionMultiply_41 {
   class=MaterialExpressionMultiply
   inputs [
    A {from=MaterialExpressionNamedRerouteUsage_56}
    B {from=MaterialExpressionConstant_12}
   ]
  }
  MaterialExpressionMultiply_42 {
   class=MaterialExpressionMultiply
   inputs [
    A {from=MaterialExpressionNamedRerouteUsage_36}
    B {from=MaterialExpressionConstant_13}
   ]
  }
  MaterialExpressionMultiply_43 {
   class=MaterialExpressionMultiply
   inputs [
    A {from=MaterialExpressionVectorParameter_3, output=A}
    B {from=MaterialExpressionScalarParameter_23}
   ]
  }
  MaterialExpressionMultiply_47 {
   class=MaterialExpressionMultiply
   inputs [
    A {from=MaterialExpressionConvert_5}
    B {from=MaterialExpressionAppendVector_0}
   ]
  }
  MaterialExpressionMultiply_5 {
   class=MaterialExpressionMultiply
   inputs [
    A {from=MaterialExpressionOneMinus_29}
    B {from=""}
   ]
   properties {constB=-1}
  }
  MaterialExpressionMultiply_50 {
   class=MaterialExpressionMultiply
   inputs [
    A {from=""}
    B {from=MaterialExpressionConstant2Vector_5}
   ]
  }
  MaterialExpressionMultiply_56 {
   class=MaterialExpressionMultiply
   inputs [
    A {from=MaterialExpressionNamedRerouteDeclaration_18}
    B {from=MaterialExpressionScalarParameter_11}
   ]
  }
  MaterialExpressionMultiply_6 {
   class=MaterialExpressionMultiply
   inputs [
    A {from=MaterialExpressionNamedRerouteUsage_23}
    B {from=MaterialExpressionNamedRerouteUsage_12}
   ]
  }
  MaterialExpressionMultiply_60 {
   class=MaterialExpressionMultiply
   inputs [
    A {from=MaterialExpressionNamedRerouteUsage_65}
    B {from=MaterialExpressionMultiply_62}
   ]
  }
  MaterialExpressionMultiply_62 {
   class=MaterialExpressionMultiply
   inputs [
    A {from=MaterialExpressionScalarParameter_12}
    B {from=""}
   ]
   properties {constB=0.25}
  }
  MaterialExpressionMultiply_64 {
   class=MaterialExpressionMultiply
   inputs [
    A {from=MaterialExpressionNamedRerouteDeclaration_23}
    B {from=MaterialExpressionNamedRerouteUsage_29}
   ]
  }
  MaterialExpressionMultiply_7 {
   class=MaterialExpressionMultiply
   inputs [
    A {from=MaterialExpressionConvert_0}
    B {from=""}
   ]
   properties {constB=2}
  }
  MaterialExpressionMultiply_8 {
   class=MaterialExpressionMultiply
   inputs [
    A {from=MaterialExpressionMultiply_6}
    B {from=MaterialExpressionLinearInterpolate_6}
   ]
  }
  MaterialExpressionMultiply_9 {
   class=MaterialExpressionMultiply
   inputs [
    A {from=MaterialExpressionNamedRerouteDeclaration_39}
    B {from=MaterialExpressionNamedRerouteDeclaration_38}
   ]
  }
  MaterialExpressionNamedRerouteDeclaration_0 {
   class=MaterialExpressionNamedRerouteDeclaration
   inputs [None {from=MaterialExpressionReroute_9, output=Output}]
   properties {name=Selected Emote Slot Mask, nodeColor=<rgba 0 0 0 1>}
  }
  MaterialExpressionNamedRerouteDeclaration_1 {
   class=MaterialExpressionNamedRerouteDeclaration
   inputs [None {from=MaterialExpressionMaterialFunctionCall_0, output=Fill}]
   properties {name=Center Background Mask, nodeColor=<rgba 0 0 0 1>}
  }
  MaterialExpressionNamedRerouteDeclaration_10 {
   class=MaterialExpressionNamedRerouteDeclaration
   inputs [None {from=MaterialExpressionScalarParameter_5}]
   properties {name=Donut Padding, nodeColor=<rgba 0 0 0 1>}
  }
  MaterialExpressionNamedRerouteDeclaration_11 {
   class=MaterialExpressionNamedRerouteDeclaration
   inputs [None {from=MaterialExpressionMultiply_43}]
   properties {name=Sub Segment Opacity, nodeColor=<rgba 0 0 0 1>}
  }
  MaterialExpressionNamedRerouteDeclaration_12 {
   class=MaterialExpressionNamedRerouteDeclaration
   inputs [None {from=MaterialExpressionVectorParameter_0, output=A}]
   properties {name=Slots Opacity, nodeColor=<rgba 0 0 0 1>}
  }
  MaterialExpressionNamedRerouteDeclaration_13 {
   class=MaterialExpressionNamedRerouteDeclaration
   inputs [None {from=MaterialExpressionVectorParameter_5, output=A}]
   properties {name=DonutBackgroundOpacity, nodeColor=<rgba 0 0 0 1>}
  }
  MaterialExpressionNamedRerouteDeclaration_14 {
   class=MaterialExpressionNamedRerouteDeclaration
   inputs [None {from=MaterialExpressionStaticSwitchParameter_0}]
   properties {name=DonutBackgroundColor, nodeColor=<rgba 0 0 0 1>}
  }
  MaterialExpressionNamedRerouteDeclaration_15 {
   class=MaterialExpressionNamedRerouteDeclaration
   inputs [None {from=MaterialExpressionDotProduct_0}]
   properties {name=Radial Segments, nodeColor=<rgba 0 0 0 1>}
  }
  MaterialExpressionNamedRerouteDeclaration_16 {
   class=MaterialExpressionNamedRerouteDeclaration
   inputs [None {from=MaterialExpressionMaterialFunctionCall_23, output=Fill}]
   properties {name=Sub Segments Mask, nodeColor=<rgba 0 0 0 1>}
  }
  MaterialExpressionNamedRerouteDeclaration_17 {
   class=MaterialExpressionNamedRerouteDeclaration
   inputs [None {from=MaterialExpressionMultiply_64}]
   properties {name=Selected Sub Segment, nodeColor=<rgba 0 0 0 1>}
  }
  MaterialExpressionNamedRerouteDeclaration_18 {
   class=MaterialExpressionNamedRerouteDeclaration
   inputs [None {from=MaterialExpressionScalarParameter_9}]
   properties {name=Sub Thickness, nodeColor=<rgba 0 0 0 1>}
  }
  MaterialExpressionNamedRerouteDeclaration_19 {
   class=MaterialExpressionNamedRerouteDeclaration
   inputs [None {from=MaterialExpressionMaterialFunctionCall_1, output=Fill}]
   properties {name=Main Segments Mask, nodeColor=<rgba 0 0 0 1>}
  }
  MaterialExpressionNamedRerouteDeclaration_2 {
   class=MaterialExpressionNamedRerouteDeclaration
   inputs [None {from=MaterialExpressionMultiply_7}]
   properties {name=Vector Length, nodeColor=<rgba 0 0 0 1>}
  }
  MaterialExpressionNamedRerouteDeclaration_20 {
   class=MaterialExpressionNamedRerouteDeclaration
   inputs [None {from=MaterialExpressionScalarParameter_4}]
   properties {name=Main Thickness, nodeColor=<rgba 0 0 0 1>}
  }
  MaterialExpressionNamedRerouteDeclaration_21 {
   class=MaterialExpressionNamedRerouteDeclaration
   inputs [None {from=MaterialExpressionMin_0}]
   properties {name=Linear Radial Segments, nodeColor=<rgba 0 0 0 1>}
  }
  MaterialExpressionNamedRerouteDeclaration_22 {
   class=MaterialExpressionNamedRerouteDeclaration
   inputs [None {from=MaterialExpressionOneMinus_26}]
   properties {name=Spacing, nodeColor=<rgba 0 0 0 1>}
  }
  MaterialExpressionNamedRerouteDeclaration_23 {
   class=MaterialExpressionNamedRerouteDeclaration
   inputs [None {from=MaterialExpressionMultiply_29}]
   properties {name=Selected Sub Segment Mask, nodeColor=<rgba 0 0 0 1>}
  }
  MaterialExpressionNamedRerouteDeclaration_24 {
   class=MaterialExpressionNamedRerouteDeclaration
   inputs [None {from=MaterialExpressionAdd_12}]
   properties {name=Spaced Radial Segments, nodeColor=<rgba 0 0 0 1>}
  }
  MaterialExpressionNamedRerouteDeclaration_25 {
   class=MaterialExpressionNamedRerouteDeclaration
   inputs [None {from=MaterialExpressionLinearInterpolate_7}]
   properties {name=Selected Slot Opacity, nodeColor=<rgba 0 0 0 1>}
  }
  MaterialExpressionNamedRerouteDeclaration_26 {
   class=MaterialExpressionNamedRerouteDeclaration
   inputs [None {from=MaterialExpressionShadingPathSwitch_0}]
   properties {name=Blur Color, nodeColor=<rgba 0 0 0 1>}
  }
  MaterialExpressionNamedRerouteDeclaration_27 {
   class=MaterialExpressionNamedRerouteDeclaration
   inputs [None {from=MaterialExpressionMultiply_27}]
   properties {name=Slots Final Opacity, nodeColor=<rgba 0 0 0 1>}
  }
  MaterialExpressionNamedRerouteDeclaration_28 {
   class=MaterialExpressionNamedRerouteDeclaration
   inputs [None {from=MaterialExpressionVectorParameter_4, output=A}]
   properties {name=Center Background Opacity, nodeColor=<rgba 0 0 0 1>}
  }
  MaterialExpressionNamedRerouteDeclaration_29 {
   class=MaterialExpressionNamedRerouteDeclaration
   inputs [None {from=MaterialExpressionMultiply_40}]
   properties {name=SubSegmentBackgroundMask, nodeColor=<rgba 0 0 0 1>}
  }
  MaterialExpressionNamedRerouteDeclaration_3 {
   class=MaterialExpressionNamedRerouteDeclaration
   inputs [None {from=MaterialExpressionMaterialFunctionCall_3, output=Normalized UV}]
   properties {name=UV, nodeColor=<rgba 0 0 0 1>}
  }
  MaterialExpressionNamedRerouteDeclaration_30 {
   class=MaterialExpressionNamedRerouteDeclaration
   inputs [None {from=MaterialExpressionVectorParameter_7, output=A}]
   properties {name=SubSegmentBackgroundOpacity, nodeColor=<rgba 0 0 0 1>}
  }
  MaterialExpressionNamedRerouteDeclaration_31 {
   class=MaterialExpressionNamedRerouteDeclaration
   inputs [None {from=MaterialExpressionVectorParameter_7, output=RGB}]
   properties {name=SubSegmentBackgroundColor, nodeColor=<rgba 0 0 0 1>}
  }
  MaterialExpressionNamedRerouteDeclaration_32 {
   class=MaterialExpressionNamedRerouteDeclaration
   inputs [None {from=MaterialExpressionDivide_0}]
   properties {name=ActiveSlotsAngle, nodeColor=<rgba 0 0 0 1>}
  }
  MaterialExpressionNamedRerouteDeclaration_33 {
   class=MaterialExpressionNamedRerouteDeclaration
   inputs [None {from=MaterialExpressionScalarParameter_1}]
   properties {name=ActiveSlots, nodeColor=<rgba 0 1 0.7188 1>}
  }
  MaterialExpressionNamedRerouteDeclaration_34 {
   class=MaterialExpressionNamedRerouteDeclaration
   inputs [None {from=MaterialExpressionMaterialFunctionCall_5, output=SDF}]
   properties {name=Donut SDF Mask, nodeColor=<rgba 0 0 0 1>}
  }
  MaterialExpressionNamedRerouteDeclaration_35 {
   class=MaterialExpressionNamedRerouteDeclaration
   inputs [None {from=MaterialExpressionMultiply_12}]
   properties {name=ActiveSlotsMask, nodeColor=<rgba 0 0 0 1>}
  }
  MaterialExpressionNamedRerouteDeclaration_36 {
   class=MaterialExpressionNamedRerouteDeclaration
   inputs [None {from=MaterialExpressionScalarParameter_8}]
   properties {name=IsActiveSlotsEnabled, nodeColor=<rgba 0 0 0 1>}
  }
  MaterialExpressionNamedRerouteDeclaration_37 {
   class=MaterialExpressionNamedRerouteDeclaration
   inputs [None {from=MaterialExpressionMaterialFunctionCall_1, output=Outline}]
   properties {name=MainSegmentsOutline, nodeColor=<rgba 0 0 0 1>}
  }
  MaterialExpressionNamedRerouteDeclaration_38 {
   class=MaterialExpressionNamedRerouteDeclaration
   inputs [None {from=MaterialExpressionReroute_14, output=Output}]
   properties {name=SlotMask0, nodeColor=<rgba 0.6875 1 0 1>}
  }
  MaterialExpressionNamedRerouteDeclaration_39 {
   class=MaterialExpressionNamedRerouteDeclaration
   inputs [None {from=MaterialExpressionVectorParameter_8, output=RGBA}]
   properties {name=SlotColor0, nodeColor=<rgba 0.9922 0 1 1>}
  }
  MaterialExpressionNamedRerouteDeclaration_4 {
   class=MaterialExpressionNamedRerouteDeclaration
   inputs [None {from=MaterialExpressionFloor_1}]
   properties {name=Subdivisions, nodeColor=<rgba 0 0 0 1>}
  }
  MaterialExpressionNamedRerouteDeclaration_40 {
   class=MaterialExpressionNamedRerouteDeclaration
   inputs [None {from=MaterialExpressionReroute_11, output=Output}]
   properties {name=SlotMask1, nodeColor=<rgba 0 0.9531 1 1>}
  }
  MaterialExpressionNamedRerouteDeclaration_41 {
   class=MaterialExpressionNamedRerouteDeclaration
   inputs [None {from=MaterialExpressionVectorParameter_10, output=RGBA}]
   properties {name=SlotColor1, nodeColor=<rgba 1 0.7266 0 1>}
  }
  MaterialExpressionNamedRerouteDeclaration_42 {
   class=MaterialExpressionNamedRerouteDeclaration
   inputs [None {from=MaterialExpressionVectorParameter_11, output=RGBA}]
   properties {name=SlotColor2, nodeColor=<rgba 0.4062 0 1 1>}
  }
  MaterialExpressionNamedRerouteDeclaration_43 {
   class=MaterialExpressionNamedRerouteDeclaration
   inputs [None {from=MaterialExpressionVectorParameter_12, output=RGBA}]
   properties {name=SlotColor3, nodeColor=<rgba 0 1 0.0859 1>}
  }
  MaterialExpressionNamedRerouteDeclaration_44 {
   class=MaterialExpressionNamedRerouteDeclaration
   inputs [None {from=MaterialExpressionVectorParameter_13, output=RGBA}]
   properties {name=SlotColor4, nodeColor=<rgba 1 0 0.2344 1>}
  }
  MaterialExpressionNamedRerouteDeclaration_45 {
   class=MaterialExpressionNamedRerouteDeclaration
   inputs [None {from=MaterialExpressionVectorParameter_14, output=RGBA}]
   properties {name=SlotColor5, nodeColor=<rgba 0 0.5547 1 1>}
  }
  MaterialExpressionNamedRerouteDeclaration_46 {
   class=MaterialExpressionNamedRerouteDeclaration
   inputs [None {from=MaterialExpressionVectorParameter_15, output=RGBA}]
   properties {name=SlotColor6, nodeColor=<rgba 0.875 1 0 1>}
  }
  MaterialExpressionNamedRerouteDeclaration_47 {
   class=MaterialExpressionNamedRerouteDeclaration
   inputs [None {from=MaterialExpressionVectorParameter_16, output=RGBA}]
   properties {name=SlotColor7, nodeColor=<rgba 0.8047 0 1 1>}
  }
  MaterialExpressionNamedRerouteDeclaration_48 {
   class=MaterialExpressionNamedRerouteDeclaration
   inputs [None {from=MaterialExpressionReroute_17, output=Output}]
   properties {name=SlotMask2, nodeColor=<rgba 0 1 0.4844 1>}
  }
  MaterialExpressionNamedRerouteDeclaration_49 {
   class=MaterialExpressionNamedRerouteDeclaration
   inputs [None {from=MaterialExpressionReroute_9, output=Output}]
   properties {name=SlotMask3, nodeColor=<rgba 1 0.1641 0 1>}
  }
  MaterialExpressionNamedRerouteDeclaration_5 {
   class=MaterialExpressionNamedRerouteDeclaration
   inputs [None {from=MaterialExpressionVectorParameter_6, output=A}]
   properties {name=DonutBackgroundOpacityLowEnd, nodeColor=<rgba 0 0 0 1>}
  }
  MaterialExpressionNamedRerouteDeclaration_50 {
   class=MaterialExpressionNamedRerouteDeclaration
   inputs [None {from=MaterialExpressionReroute_11, output=Output}]
   properties {name=SlotMask4, nodeColor=<rgba 0 0.1562 1 1>}
  }
  MaterialExpressionNamedRerouteDeclaration_51 {
   class=MaterialExpressionNamedRerouteDeclaration
   inputs [None {from=MaterialExpressionReroute_8, output=Output}]
   properties {name=SlotMask5, nodeColor=<rgba 0.4766 1 0 1>}
  }
  MaterialExpressionNamedRerouteDeclaration_52 {
   class=MaterialExpressionNamedRerouteDeclaration
   inputs [None {from=MaterialExpressionReroute_16, output=Output}]
   properties {name=SlotMask6, nodeColor=<rgba 1 0 0.7969 1>}
  }
  MaterialExpressionNamedRerouteDeclaration_53 {
   class=MaterialExpressionNamedRerouteDeclaration
   inputs [None {from=MaterialExpressionReroute_14, output=Output}]
   properties {name=SlotMask7, nodeColor=<rgba 0 1 0.8828 1>}
  }
  MaterialExpressionNamedRerouteDeclaration_54 {
   class=MaterialExpressionNamedRerouteDeclaration
   inputs [None {from=MaterialExpressionMultiply_9}]
   properties {name=SlotMaskColor0, nodeColor=<rgba 1 0.5625 0 1>}
  }
  MaterialExpressionNamedRerouteDeclaration_55 {
   class=MaterialExpressionNamedRerouteDeclaration
   inputs [None {from=MaterialExpressionMultiply_14}]
   properties {name=SlotMaskColor1, nodeColor=<rgba 0.2422 0 1 1>}
  }
  MaterialExpressionNamedRerouteDeclaration_56 {
   class=MaterialExpressionNamedRerouteDeclaration
   inputs [None {from=MaterialExpressionMultiply_15}]
   properties {name=SlotMaskColor2, nodeColor=<rgba 0.0781 1 0 1>}
  }
  MaterialExpressionNamedRerouteDeclaration_57 {
   class=MaterialExpressionNamedRerouteDeclaration
   inputs [None {from=MaterialExpressionMultiply_16}]
   properties {name=SlotMaskColor3, nodeColor=<rgba 1 0 0.3984 1>}
  }
  MaterialExpressionNamedRerouteDeclaration_58 {
   class=MaterialExpressionNamedRerouteDeclaration
   inputs [None {from=MaterialExpressionMultiply_17}]
   properties {name=SlotMaskColor4, nodeColor=<rgba 0 0.7188 1 1>}
  }
  MaterialExpressionNamedRerouteDeclaration_59 {
   class=MaterialExpressionNamedRerouteDeclaration
   inputs [None {from=MaterialExpressionMultiply_24}]
   properties {name=SlotMaskColor5, nodeColor=<rgba 1 0.9609 0 1>}
  }
  MaterialExpressionNamedRerouteDeclaration_6 {
   class=MaterialExpressionNamedRerouteDeclaration
   inputs [None {from=MaterialExpressionMaterialFunctionCall_5, output=Outline}]
   properties {name=Donut Background Mask, nodeColor=<rgba 0 0 0 1>}
  }
  MaterialExpressionNamedRerouteDeclaration_60 {
   class=MaterialExpressionNamedRerouteDeclaration
   inputs [None {from=MaterialExpressionMultiply_25}]
   properties {name=SlotMaskColor6, nodeColor=<rgba 0.6406 0 1 1>}
  }
  MaterialExpressionNamedRerouteDeclaration_61 {
   class=MaterialExpressionNamedRerouteDeclaration
   inputs [None {from=MaterialExpressionMultiply_26}]
   properties {name=SlotMaskColor7, nodeColor=<rgba 0 1 0.3203 1>}
  }
  MaterialExpressionNamedRerouteDeclaration_62 {
   class=MaterialExpressionNamedRerouteDeclaration
   inputs [None {from=MaterialExpressionMax_13}]
   properties {name=SlotsCombinedColor, nodeColor=<rgba 1 0 0 1>}
  }
  MaterialExpressionNamedRerouteDeclaration_63 {
   class=MaterialExpressionNamedRerouteDeclaration
   inputs [None {from=MaterialExpressionConvert_3}]
   properties {name=SlotsCombinedColorRGB, nodeColor=<rgba 0.2422 0 1 1>}
  }
  MaterialExpressionNamedRerouteDeclaration_64 {
   class=MaterialExpressionNamedRerouteDeclaration
   inputs [None {from=MaterialExpressionScalarParameter_13}]
   properties {name=UseSlotColors, nodeColor=<rgba 0 0.5781 1 1>}
  }
  MaterialExpressionNamedRerouteDeclaration_65 {
   class=MaterialExpressionNamedRerouteDeclaration
   inputs [None {from=MaterialExpressionStep_1}]
   properties {name=MaskedOutSlots, nodeColor=<rgba 0 1 0.0156 1>}
  }
  MaterialExpressionNamedRerouteDeclaration_66 {
   class=MaterialExpressionNamedRerouteDeclaration
   inputs [None {from=MaterialExpressionMultiply_35}]
   properties {name=SlotHighlightColor, nodeColor=<rgba 0 0.8594 1 1>}
  }
  MaterialExpressionNamedRerouteDeclaration_67 {
   class=MaterialExpressionNamedRerouteDeclaration
   inputs [None {from=MaterialExpressionMultiply_36}]
   properties {name=SlotHighlightMask, nodeColor=<rgba 1 0.8203 0 1>}
  }
  MaterialExpressionNamedRerouteDeclaration_7 {
   class=MaterialExpressionNamedRerouteDeclaration
   inputs [None {from=MaterialExpressionLinearInterpolate_5}]
   properties {name=Selected Slot Visual, nodeColor=<rgba 0 0 0 1>}
  }
  MaterialExpressionNamedRerouteDeclaration_8 {
   class=MaterialExpressionNamedRerouteDeclaration
   inputs [None {from=MaterialExpressionLinearInterpolate_12}]
   properties {name=Slots Visuals, nodeColor=<rgba 0 0 0 1>}
  }
  MaterialExpressionNamedRerouteDeclaration_9 {
   class=MaterialExpressionNamedRerouteDeclaration
   inputs [None {from=MaterialExpressionMultiply_0}]
   properties {name=Subdivised Angle, nodeColor=<rgba 0 0 0 1>}
  }
  MaterialExpressionNamedRerouteUsage_0 {class=MaterialExpressionNamedRerouteUsage}
  MaterialExpressionNamedRerouteUsage_1 {class=MaterialExpressionNamedRerouteUsage}
  MaterialExpressionNamedRerouteUsage_10 {class=MaterialExpressionNamedRerouteUsage}
  MaterialExpressionNamedRerouteUsage_11 {class=MaterialExpressionNamedRerouteUsage}
  MaterialExpressionNamedRerouteUsage_12 {class=MaterialExpressionNamedRerouteUsage}
  MaterialExpressionNamedRerouteUsage_13 {class=MaterialExpressionNamedRerouteUsage}
  MaterialExpressionNamedRerouteUsage_14 {class=MaterialExpressionNamedRerouteUsage}
  MaterialExpressionNamedRerouteUsage_15 {class=MaterialExpressionNamedRerouteUsage}
  MaterialExpressionNamedRerouteUsage_17 {class=MaterialExpressionNamedRerouteUsage}
  MaterialExpressionNamedRerouteUsage_18 {class=MaterialExpressionNamedRerouteUsage}
  MaterialExpressionNamedRerouteUsage_19 {class=MaterialExpressionNamedRerouteUsage}
  MaterialExpressionNamedRerouteUsage_2 {class=MaterialExpressionNamedRerouteUsage}
  MaterialExpressionNamedRerouteUsage_20 {class=MaterialExpressionNamedRerouteUsage}
  MaterialExpressionNamedRerouteUsage_21 {class=MaterialExpressionNamedRerouteUsage}
  MaterialExpressionNamedRerouteUsage_22 {class=MaterialExpressionNamedRerouteUsage}
  MaterialExpressionNamedRerouteUsage_23 {class=MaterialExpressionNamedRerouteUsage}
  MaterialExpressionNamedRerouteUsage_24 {class=MaterialExpressionNamedRerouteUsage}
  MaterialExpressionNamedRerouteUsage_25 {class=MaterialExpressionNamedRerouteUsage}
  MaterialExpressionNamedRerouteUsage_26 {class=MaterialExpressionNamedRerouteUsage}
  MaterialExpressionNamedRerouteUsage_27 {class=MaterialExpressionNamedRerouteUsage}
  MaterialExpressionNamedRerouteUsage_28 {class=MaterialExpressionNamedRerouteUsage}
  MaterialExpressionNamedRerouteUsage_29 {class=MaterialExpressionNamedRerouteUsage}
  MaterialExpressionNamedRerouteUsage_3 {class=MaterialExpressionNamedRerouteUsage}
  MaterialExpressionNamedRerouteUsage_30 {class=MaterialExpressionNamedRerouteUsage}
  MaterialExpressionNamedRerouteUsage_31 {class=MaterialExpressionNamedRerouteUsage}
  MaterialExpressionNamedRerouteUsage_32 {class=MaterialExpressionNamedRerouteUsage}
  MaterialExpressionNamedRerouteUsage_33 {class=MaterialExpressionNamedRerouteUsage}
  MaterialExpressionNamedRerouteUsage_34 {class=MaterialExpressionNamedRerouteUsage}
  MaterialExpressionNamedRerouteUsage_35 {class=MaterialExpressionNamedRerouteUsage}
  MaterialExpressionNamedRerouteUsage_36 {class=MaterialExpressionNamedRerouteUsage}
  MaterialExpressionNamedRerouteUsage_37 {class=MaterialExpressionNamedRerouteUsage}
  MaterialExpressionNamedRerouteUsage_38 {class=MaterialExpressionNamedRerouteUsage}
  MaterialExpressionNamedRerouteUsage_39 {class=MaterialExpressionNamedRerouteUsage}
  MaterialExpressionNamedRerouteUsage_4 {class=MaterialExpressionNamedRerouteUsage}
  MaterialExpressionNamedRerouteUsage_40 {class=MaterialExpressionNamedRerouteUsage}
  MaterialExpressionNamedRerouteUsage_41 {class=MaterialExpressionNamedRerouteUsage}
  MaterialExpressionNamedRerouteUsage_42 {class=MaterialExpressionNamedRerouteUsage}
  MaterialExpressionNamedRerouteUsage_43 {class=MaterialExpressionNamedRerouteUsage}
  MaterialExpressionNamedRerouteUsage_44 {class=MaterialExpressionNamedRerouteUsage}
  MaterialExpressionNamedRerouteUsage_45 {class=MaterialExpressionNamedRerouteUsage}
  MaterialExpressionNamedRerouteUsage_46 {class=MaterialExpressionNamedRerouteUsage}
  MaterialExpressionNamedRerouteUsage_47 {class=MaterialExpressionNamedRerouteUsage}
  MaterialExpressionNamedRerouteUsage_48 {class=MaterialExpressionNamedRerouteUsage}
  MaterialExpressionNamedRerouteUsage_49 {class=MaterialExpressionNamedRerouteUsage}
  MaterialExpressionNamedRerouteUsage_5 {class=MaterialExpressionNamedRerouteUsage}
  MaterialExpressionNamedRerouteUsage_50 {class=MaterialExpressionNamedRerouteUsage}
  MaterialExpressionNamedRerouteUsage_51 {class=MaterialExpressionNamedRerouteUsage}
  MaterialExpressionNamedRerouteUsage_52 {class=MaterialExpressionNamedRerouteUsage}
  MaterialExpressionNamedRerouteUsage_53 {class=MaterialExpressionNamedRerouteUsage}
  MaterialExpressionNamedRerouteUsage_54 {class=MaterialExpressionNamedRerouteUsage}
  MaterialExpressionNamedRerouteUsage_55 {class=MaterialExpressionNamedRerouteUsage}
  MaterialExpressionNamedRerouteUsage_56 {class=MaterialExpressionNamedRerouteUsage}
  MaterialExpressionNamedRerouteUsage_57 {class=MaterialExpressionNamedRerouteUsage}
  MaterialExpressionNamedRerouteUsage_58 {class=MaterialExpressionNamedRerouteUsage}
  MaterialExpressionNamedRerouteUsage_6 {class=MaterialExpressionNamedRerouteUsage}
  MaterialExpressionNamedRerouteUsage_60 {class=MaterialExpressionNamedRerouteUsage}
  MaterialExpressionNamedRerouteUsage_63 {class=MaterialExpressionNamedRerouteUsage}
  MaterialExpressionNamedRerouteUsage_64 {class=MaterialExpressionNamedRerouteUsage}
  MaterialExpressionNamedRerouteUsage_65 {class=MaterialExpressionNamedRerouteUsage}
  MaterialExpressionNamedRerouteUsage_66 {class=MaterialExpressionNamedRerouteUsage}
  MaterialExpressionNamedRerouteUsage_67 {class=MaterialExpressionNamedRerouteUsage}
  MaterialExpressionNamedRerouteUsage_68 {class=MaterialExpressionNamedRerouteUsage}
  MaterialExpressionNamedRerouteUsage_69 {class=MaterialExpressionNamedRerouteUsage}
  MaterialExpressionNamedRerouteUsage_7 {class=MaterialExpressionNamedRerouteUsage}
  MaterialExpressionNamedRerouteUsage_70 {class=MaterialExpressionNamedRerouteUsage}
  MaterialExpressionNamedRerouteUsage_71 {class=MaterialExpressionNamedRerouteUsage}
  MaterialExpressionNamedRerouteUsage_75 {class=MaterialExpressionNamedRerouteUsage}
  MaterialExpressionNamedRerouteUsage_76 {class=MaterialExpressionNamedRerouteUsage}
  MaterialExpressionNamedRerouteUsage_77 {class=MaterialExpressionNamedRerouteUsage}
  MaterialExpressionNamedRerouteUsage_78 {class=MaterialExpressionNamedRerouteUsage}
  MaterialExpressionNamedRerouteUsage_79 {class=MaterialExpressionNamedRerouteUsage}
  MaterialExpressionNamedRerouteUsage_8 {class=MaterialExpressionNamedRerouteUsage}
  MaterialExpressionNamedRerouteUsage_80 {class=MaterialExpressionNamedRerouteUsage}
  MaterialExpressionNamedRerouteUsage_81 {class=MaterialExpressionNamedRerouteUsage}
  MaterialExpressionNamedRerouteUsage_82 {class=MaterialExpressionNamedRerouteUsage}
  MaterialExpressionNamedRerouteUsage_83 {class=MaterialExpressionNamedRerouteUsage}
  MaterialExpressionNamedRerouteUsage_84 {class=MaterialExpressionNamedRerouteUsage}
  MaterialExpressionNamedRerouteUsage_85 {class=MaterialExpressionNamedRerouteUsage}
  MaterialExpressionNamedRerouteUsage_86 {class=MaterialExpressionNamedRerouteUsage}
  MaterialExpressionNamedRerouteUsage_87 {class=MaterialExpressionNamedRerouteUsage}
  MaterialExpressionNamedRerouteUsage_9 {class=MaterialExpressionNamedRerouteUsage}
  MaterialExpressionOneMinus_0 {class=MaterialExpressionOneMinus, inputs [None {from=MaterialExpressionSaturate_3}]}
  MaterialExpressionOneMinus_1 {class=MaterialExpressionOneMinus, inputs [None {from=MaterialExpressionNamedRerouteUsage_2}]}
  MaterialExpressionOneMinus_10 {class=MaterialExpressionOneMinus, inputs [None {from=MaterialExpressionSaturate_8}]}
  MaterialExpressionOneMinus_11 {class=MaterialExpressionOneMinus, inputs [None {from=MaterialExpressionSaturate_9}]}
  MaterialExpressionOneMinus_12 {class=MaterialExpressionOneMinus, inputs [None {from=MaterialExpressionFrac_3}]}
  MaterialExpressionOneMinus_13 {class=MaterialExpressionOneMinus, inputs [None {from=MaterialExpressionSaturate_10}]}
  MaterialExpressionOneMinus_14 {class=MaterialExpressionOneMinus, inputs [None {from=MaterialExpressionSaturate_11}]}
  MaterialExpressionOneMinus_15 {class=MaterialExpressionOneMinus, inputs [None {from=MaterialExpressionSaturate_12}]}
  MaterialExpressionOneMinus_16 {class=MaterialExpressionOneMinus, inputs [None {from=MaterialExpressionMultiply_31}]}
  MaterialExpressionOneMinus_17 {class=MaterialExpressionOneMinus, inputs [None {from=MaterialExpressionScalarParameter_10}]}
  MaterialExpressionOneMinus_18 {class=MaterialExpressionOneMinus, inputs [None {from=MaterialExpressionScalarParameter_14}]}
  MaterialExpressionOneMinus_19 {class=MaterialExpressionOneMinus, inputs [None {from=MaterialExpressionScalarParameter_16}]}
  MaterialExpressionOneMinus_2 {class=MaterialExpressionOneMinus, inputs [None {from=MaterialExpressionAdd_1}]}
  MaterialExpressionOneMinus_20 {class=MaterialExpressionOneMinus, inputs [None {from=MaterialExpressionScalarParameter_17}]}
  MaterialExpressionOneMinus_21 {class=MaterialExpressionOneMinus, inputs [None {from=MaterialExpressionScalarParameter_18}]}
  MaterialExpressionOneMinus_22 {class=MaterialExpressionOneMinus, inputs [None {from=MaterialExpressionScalarParameter_19}]}
  MaterialExpressionOneMinus_23 {class=MaterialExpressionOneMinus, inputs [None {from=MaterialExpressionNamedRerouteUsage_51}]}
  MaterialExpressionOneMinus_24 {class=MaterialExpressionOneMinus, inputs [None {from=MaterialExpressionNamedRerouteUsage_58}]}
  MaterialExpressionOneMinus_25 {class=MaterialExpressionOneMinus, inputs [None {from=MaterialExpressionScalarParameter_20}]}
  MaterialExpressionOneMinus_26 {class=MaterialExpressionOneMinus, inputs [None {from=MaterialExpressionMultiply_10}]}
  MaterialExpressionOneMinus_27 {class=MaterialExpressionOneMinus, inputs [None {from=MaterialExpressionScalarParameter_21}]}
  MaterialExpressionOneMinus_29 {class=MaterialExpressionOneMinus, inputs [None {from=MaterialExpressionNamedRerouteUsage_66}]}
  MaterialExpressionOneMinus_3 {class=MaterialExpressionOneMinus, inputs [None {from=MaterialExpressionSaturate_4}]}
  MaterialExpressionOneMinus_4 {class=MaterialExpressionOneMinus, inputs [None {from=MaterialExpressionConvert_1}]}
  MaterialExpressionOneMinus_5 {class=MaterialExpressionOneMinus, inputs [None {from=MaterialExpressionSaturate_7}]}
  MaterialExpressionOneMinus_6 {class=MaterialExpressionOneMinus, inputs [None {from=MaterialExpressionSaturate_5}]}
  MaterialExpressionOneMinus_7 {class=MaterialExpressionOneMinus, inputs [None {from=MaterialExpressionMax_6}]}
  MaterialExpressionOneMinus_8 {class=MaterialExpressionOneMinus, inputs [None {from=MaterialExpressionCeil_6}]}
  MaterialExpressionOneMinus_9 {class=MaterialExpressionOneMinus, inputs [None {from=MaterialExpressionSaturate_6}]}
  MaterialExpressionPinBase_0 {
   class=MaterialExpressionPinBase
   inputs []
   properties {reroutePins [
     SubdivisionAngle {expression=&self:MaterialGraph_0.MaterialGraphNode_Composite_1.GetSegmentByIndex.MaterialExpressionReroute_7}
     Index {expression=&self:MaterialGraph_0.MaterialGraphNode_Composite_1.GetSegmentByIndex.MaterialExpressionReroute_8}
    ]}
  }
  MaterialExpressionPinBase_1 {
   class=MaterialExpressionPinBase
   inputs [Output {from=MaterialExpressionOneMinus_5, output=""}]
   properties {reroutePins [Output {expression=&self:MaterialGraph_0.MaterialGraphNode_Composite_1.GetSegmentByIndex.MaterialExpressionReroute_9}]}
  }
  MaterialExpressionPinBase_10 {
   class=MaterialExpressionPinBase
   inputs []
   properties {reroutePins [
     SubdivisionAngle {expression=&self:MaterialExpressionReroute_13}
     Index {expression=&self:MaterialExpressionReroute_6}
    ]}
  }
  MaterialExpressionPinBase_11 {
   class=MaterialExpressionPinBase
   inputs [Output {from=MaterialExpressionOneMinus_11, output=""}]
   properties {reroutePins [Output {expression=&self:MaterialExpressionReroute_14}]}
  }
  MaterialExpressionPinBase_12 {
   class=MaterialExpressionPinBase
   inputs []
   properties {reroutePins [
     SubdivisionAngle {expression=&self:MaterialExpressionReroute_9}
     Index {expression=&self:MaterialExpressionReroute_15}
    ]}
  }
  MaterialExpressionPinBase_13 {
   class=MaterialExpressionPinBase
   inputs [Output {from=MaterialExpressionOneMinus_13, output=""}]
   properties {reroutePins [Output {expression=&self:MaterialExpressionReroute_16}]}
  }
  MaterialExpressionPinBase_14 {
   class=MaterialExpressionPinBase
   inputs []
   properties {reroutePins [
     SubdivisionAngle {expression=&self:MaterialExpressionReroute_17}
     Index {expression=&self:MaterialExpressionReroute_7}
    ]}
  }
  MaterialExpressionPinBase_15 {
   class=MaterialExpressionPinBase
   inputs [Output {from=MaterialExpressionOneMinus_14, output=""}]
   properties {reroutePins [Output {expression=&self:MaterialExpressionReroute_8}]}
  }
  MaterialExpressionPinBase_16 {
   class=MaterialExpressionPinBase
   inputs []
   properties {reroutePins [
     SubdivisionAngle {expression=&self:MaterialExpressionReroute_18}
     Index {expression=&self:MaterialExpressionReroute_10}
    ]}
  }
  MaterialExpressionPinBase_17 {
   class=MaterialExpressionPinBase
   inputs [Output {from=MaterialExpressionOneMinus_15, output=""}]
   properties {reroutePins [Output {expression=&self:MaterialExpressionReroute_11}]}
  }
  MaterialExpressionPinBase_2 {
   class=MaterialExpressionPinBase
   inputs []
   properties {reroutePins [
     SubdivisionAngle {expression=&self:MaterialGraph_0.MaterialGraphNode_Composite_10.GetSegmentByIndex.MaterialExpressionReroute_12}
     Index {expression=&self:MaterialGraph_0.MaterialGraphNode_Composite_10.GetSegmentByIndex.MaterialExpressionReroute_13}
    ]}
  }
  MaterialExpressionPinBase_3 {
   class=MaterialExpressionPinBase
   inputs [Output {from=MaterialExpressionOneMinus_3, output=""}]
   properties {reroutePins [Output {expression=&self:MaterialGraph_0.MaterialGraphNode_Composite_10.GetSegmentByIndex.MaterialExpressionReroute_14}]}
  }
  MaterialExpressionPinBase_4 {
   class=MaterialExpressionPinBase
   inputs []
   properties {reroutePins [
     SubdivisionAngle {expression=&self:MaterialGraph_0.MaterialGraphNode_Composite_9.GetSegmentByIndex.MaterialExpressionReroute_9}
     Index {expression=&self:MaterialGraph_0.MaterialGraphNode_Composite_9.GetSegmentByIndex.MaterialExpressionReroute_10}
    ]}
  }
  MaterialExpressionPinBase_5 {
   class=MaterialExpressionPinBase
   inputs [Output {from=MaterialExpressionOneMinus_6, output=""}]
   properties {reroutePins [Output {expression=&self:MaterialGraph_0.MaterialGraphNode_Composite_9.GetSegmentByIndex.MaterialExpressionReroute_11}]}
  }
  MaterialExpressionPinBase_6 {
   class=MaterialExpressionPinBase
   inputs []
   properties {reroutePins [
     SubdivisionAngle {expression=&self:MaterialGraph_0.MaterialGraphNode_Composite_8.GetSegmentByIndex.MaterialExpressionReroute_15}
     Index {expression=&self:MaterialGraph_0.MaterialGraphNode_Composite_8.GetSegmentByIndex.MaterialExpressionReroute_16}
    ]}
  }
  MaterialExpressionPinBase_7 {
   class=MaterialExpressionPinBase
   inputs [Output {from=MaterialExpressionOneMinus_9, output=""}]
   properties {reroutePins [Output {expression=&self:MaterialGraph_0.MaterialGraphNode_Composite_8.GetSegmentByIndex.MaterialExpressionReroute_17}]}
  }
  MaterialExpressionPinBase_8 {
   class=MaterialExpressionPinBase
   inputs []
   properties {reroutePins [
     SubdivisionAngle {expression=&self:MaterialGraph_0.MaterialGraphNode_Composite_7.GetSegmentByIndex.MaterialExpressionReroute_7}
     Index {expression=&self:MaterialGraph_0.MaterialGraphNode_Composite_7.GetSegmentByIndex.MaterialExpressionReroute_8}
    ]}
  }
  MaterialExpressionPinBase_9 {
   class=MaterialExpressionPinBase
   inputs [Output {from=MaterialExpressionOneMinus_10, output=""}]
   properties {reroutePins [Output {expression=&self:MaterialGraph_0.MaterialGraphNode_Composite_7.GetSegmentByIndex.MaterialExpressionReroute_9}]}
  }
  MaterialExpressionQualitySwitch_0 {
   class=MaterialExpressionQualitySwitch
   inputs [
    Default {from=MaterialExpressionConstant_7, output=""}
    Low {from=MaterialExpressionConstant_7, output=""}
    High {from=MaterialExpressionConstant_7, output=""}
    Medium {from=MaterialExpressionConstant_7, output=""}
    Epic {from=MaterialExpressionMaterialFunctionCall_7, output=LinearRGB}
   ]
   properties {}
  }
  MaterialExpressionQualitySwitch_1 {
   class=MaterialExpressionQualitySwitch
   inputs [
    Default {from=MaterialExpressionMultiply_41, output=""}
    Low {from=MaterialExpressionMultiply_41, output=""}
    High {from=MaterialExpressionMultiply_41, output=""}
    Medium {from=MaterialExpressionMultiply_41, output=""}
    Epic {from=MaterialExpressionNamedRerouteUsage_56, output=""}
   ]
   properties {}
  }
  MaterialExpressionQualitySwitch_2 {
   class=MaterialExpressionQualitySwitch
   inputs [
    Default {from=MaterialExpressionMultiply_42, output=""}
    Low {from=MaterialExpressionMultiply_42, output=""}
    High {from=MaterialExpressionMultiply_42, output=""}
    Medium {from=MaterialExpressionMultiply_42, output=""}
    Epic {from=MaterialExpressionNamedRerouteUsage_36, output=""}
   ]
   properties {}
  }
  MaterialExpressionQualitySwitch_3 {
   class=MaterialExpressionQualitySwitch
   inputs [
    Default {from=MaterialExpressionVectorParameter_6, output=RGB}
    Low {from=MaterialExpressionVectorParameter_6, output=RGB}
    High {from=MaterialExpressionVectorParameter_6, output=RGB}
    Medium {from=MaterialExpressionVectorParameter_6, output=RGB}
    Epic {from=MaterialExpressionNamedRerouteDeclaration_14, output=""}
   ]
   properties {}
  }
  MaterialExpressionQualitySwitch_4 {
   class=MaterialExpressionQualitySwitch
   inputs [
    Default {from=MaterialExpressionNamedRerouteUsage_3, output=""}
    Low {from=MaterialExpressionNamedRerouteUsage_3, output=""}
    High {from=MaterialExpressionNamedRerouteUsage_3, output=""}
    Medium {from=MaterialExpressionNamedRerouteUsage_3, output=""}
    Epic {from=MaterialExpressionStaticSwitchParameter_3, output=""}
   ]
   properties {}
  }
  MaterialExpressionReroute_0 {class=MaterialExpressionReroute, inputs [None {from=MaterialExpressionConstant_7}]}
  MaterialExpressionReroute_1 {class=MaterialExpressionReroute, inputs [None {from=MaterialExpressionNamedRerouteUsage_86}]}
  MaterialExpressionReroute_10 {class=MaterialExpressionReroute, inputs [None {from=MaterialExpressionConstant_14}]}
  MaterialExpressionReroute_11 {class=MaterialExpressionReroute, inputs [None {from=MaterialExpressionOneMinus_15}]}
  MaterialExpressionReroute_13 {class=MaterialExpressionReroute, inputs [None {from=MaterialExpressionNamedRerouteUsage_37}]}
  MaterialExpressionReroute_14 {class=MaterialExpressionReroute, inputs [None {from=MaterialExpressionOneMinus_11}]}
  MaterialExpressionReroute_15 {class=MaterialExpressionReroute, inputs [None {from=MaterialExpressionConstant_10}]}
  MaterialExpressionReroute_16 {class=MaterialExpressionReroute, inputs [None {from=MaterialExpressionOneMinus_13}]}
  MaterialExpressionReroute_17 {class=MaterialExpressionReroute, inputs [None {from=MaterialExpressionNamedRerouteUsage_32}]}
  MaterialExpressionReroute_18 {class=MaterialExpressionReroute, inputs [None {from=MaterialExpressionNamedRerouteUsage_30}]}
  MaterialExpressionReroute_19 {class=MaterialExpressionReroute, inputs [None {from=MaterialExpressionShadingPathSwitch_3}]}
  MaterialExpressionReroute_2 {class=MaterialExpressionReroute, inputs [None {from=MaterialExpressionNamedRerouteUsage_31}]}
  MaterialExpressionReroute_20 {class=MaterialExpressionReroute, inputs [None {from=MaterialExpressionReroute_19}]}
  MaterialExpressionReroute_22 {class=MaterialExpressionReroute, inputs [None {from=MaterialExpressionLinearInterpolate_11}]}
  MaterialExpressionReroute_3 {class=MaterialExpressionReroute, inputs [None {from=MaterialExpressionNamedRerouteUsage_14}]}
  MaterialExpressionReroute_4 {class=MaterialExpressionReroute, inputs [None {from=MaterialExpressionNamedRerouteUsage_35}]}
  MaterialExpressionReroute_5 {class=MaterialExpressionReroute, inputs [None {from=MaterialExpressionNamedRerouteUsage_35}]}
  MaterialExpressionReroute_6 {class=MaterialExpressionReroute, inputs [None {from=MaterialExpressionConstant_9}]}
  MaterialExpressionReroute_7 {class=MaterialExpressionReroute, inputs [None {from=MaterialExpressionConstant_11}]}
  MaterialExpressionReroute_8 {class=MaterialExpressionReroute, inputs [None {from=MaterialExpressionOneMinus_14}]}
  MaterialExpressionReroute_9 {class=MaterialExpressionReroute, inputs [None {from=MaterialExpressionNamedRerouteUsage_34}]}
  MaterialExpressionRotator_0 {
   class=MaterialExpressionRotator
   inputs [
    Coordinate {from=MaterialExpressionAppendVector_7, output=""}
    Time {from=MaterialExpressionDivide_2, output=""}
   ]
   properties {speed=3.1416}
  }
  MaterialExpressionSaturate_0 {class=MaterialExpressionSaturate, inputs [None {from=MaterialExpressionOneMinus_2}]}
  MaterialExpressionSaturate_1 {class=MaterialExpressionSaturate, inputs [None {from=MaterialExpressionMultiply_11}]}
  MaterialExpressionSaturate_10 {class=MaterialExpressionSaturate, inputs [None {from=MaterialExpressionAbs_7}]}
  MaterialExpressionSaturate_11 {class=MaterialExpressionSaturate, inputs [None {from=MaterialExpressionAbs_8}]}
  MaterialExpressionSaturate_12 {class=MaterialExpressionSaturate, inputs [None {from=MaterialExpressionAbs_9}]}
  MaterialExpressionSaturate_13 {class=MaterialExpressionSaturate, inputs [None {from=MaterialExpressionMultiply_32}]}
  MaterialExpressionSaturate_14 {class=MaterialExpressionSaturate, inputs [None {from=MaterialExpressionMultiply_37}]}
  MaterialExpressionSaturate_16 {class=MaterialExpressionSaturate, inputs [None {from=MaterialExpressionNamedRerouteUsage_22}]}
  MaterialExpressionSaturate_17 {class=MaterialExpressionSaturate, inputs [None {from=MaterialExpressionMultiply_36}]}
  MaterialExpressionSaturate_2 {class=MaterialExpressionSaturate, inputs [None {from=MaterialExpressionAdd_14}]}
  MaterialExpressionSaturate_3 {class=MaterialExpressionSaturate, inputs [None {from=MaterialExpressionAbs_0}]}
  MaterialExpressionSaturate_4 {class=MaterialExpressionSaturate, inputs [None {from=MaterialExpressionAbs_2}]}
  MaterialExpressionSaturate_5 {class=MaterialExpressionSaturate, inputs [None {from=MaterialExpressionAbs_3}]}
  MaterialExpressionSaturate_6 {class=MaterialExpressionSaturate, inputs [None {from=MaterialExpressionAbs_4}]}
  MaterialExpressionSaturate_7 {class=MaterialExpressionSaturate, inputs [None {from=MaterialExpressionAbs_1}]}
  MaterialExpressionSaturate_8 {class=MaterialExpressionSaturate, inputs [None {from=MaterialExpressionAbs_5}]}
  MaterialExpressionSaturate_9 {class=MaterialExpressionSaturate, inputs [None {from=MaterialExpressionAbs_6}]}
  MaterialExpressionScalarParameter_0 {class=MaterialExpressionScalarParameter, properties {defaultValue=8, group=Settings, parameterName=Subdivisions}}
  MaterialExpressionScalarParameter_1 {class=MaterialExpressionScalarParameter, properties {defaultValue=4, parameterName=ActiveSlots}}
  MaterialExpressionScalarParameter_10 {class=MaterialExpressionScalarParameter, properties {defaultValue=1, group=SlotColors, parameterName=IsHighlightSlot6}}
  MaterialExpressionScalarParameter_11 {class=MaterialExpressionScalarParameter, properties {defaultValue=0.75, group=Sub Segments, parameterName=SubCornerRadius, sliderMax=1}}
  MaterialExpressionScalarParameter_12 {
   class=MaterialExpressionScalarParameter
   properties {defaultValue=0.2, desc=Segment Spacing, group=Settings, parameterName=SegmentSpacing, sliderMax=1}
  }
  MaterialExpressionScalarParameter_13 {class=MaterialExpressionScalarParameter, properties {group=Settings, parameterName=UseSlotColors}}
  MaterialExpressionScalarParameter_14 {class=MaterialExpressionScalarParameter, properties {defaultValue=1, group=SlotColors, parameterName=IsHighlightSlot7}}
  MaterialExpressionScalarParameter_16 {class=MaterialExpressionScalarParameter, properties {defaultValue=1, group=SlotColors, parameterName=IsHighlightSlot5}}
  MaterialExpressionScalarParameter_17 {class=MaterialExpressionScalarParameter, properties {defaultValue=1, group=SlotColors, parameterName=IsHighlightSlot4}}
  MaterialExpressionScalarParameter_18 {class=MaterialExpressionScalarParameter, properties {defaultValue=1, group=SlotColors, parameterName=IsHighlightSlot3}}
  MaterialExpressionScalarParameter_19 {class=MaterialExpressionScalarParameter, properties {defaultValue=1, group=SlotColors, parameterName=IsHighlightSlot2}}
  MaterialExpressionScalarParameter_2 {class=MaterialExpressionScalarParameter, properties {defaultValue=0.025, parameterName=OutlineThickness}}
  MaterialExpressionScalarParameter_20 {class=MaterialExpressionScalarParameter, properties {defaultValue=1, group=SlotColors, parameterName=IsHighlightSlot1}}
  MaterialExpressionScalarParameter_21 {class=MaterialExpressionScalarParameter, properties {group=SlotColors, parameterName=IsHighlightSlot0}}
  MaterialExpressionScalarParameter_22 {class=MaterialExpressionScalarParameter, properties {defaultValue=0.8, parameterName=GlowScale}}
  MaterialExpressionScalarParameter_23 {class=MaterialExpressionScalarParameter, properties {defaultValue=1, parameterName=SubSegmentOpacity}}
  MaterialExpressionScalarParameter_3 {class=MaterialExpressionScalarParameter, properties {desc=Index, group=Settings, parameterName=SelectedIndex}}
  MaterialExpressionScalarParameter_4 {
   class=MaterialExpressionScalarParameter
   properties {defaultValue=0.375, desc=Ring Size, group=Main Segments, parameterName=MainRingThickness, sliderMax=1}
  }
  MaterialExpressionScalarParameter_5 {
   class=MaterialExpressionScalarParameter
   properties {defaultValue=0.025, desc=Ring Size, group=Settings, parameterName=DonutPadding, sliderMax=1}
  }
  MaterialExpressionScalarParameter_6 {class=MaterialExpressionScalarParameter, properties {defaultValue=0.65, parameterName=ActiveSlotsOpacity}}
  MaterialExpressionScalarParameter_7 {class=MaterialExpressionScalarParameter, properties {defaultValue=0.15, group=Main Segments, parameterName=MainCornerRadius, sliderMax=1}}
  MaterialExpressionScalarParameter_8 {class=MaterialExpressionScalarParameter, properties {parameterName=IsActiveSlotsEnabled}}
  MaterialExpressionScalarParameter_9 {class=MaterialExpressionScalarParameter, properties {defaultValue=0.0375, group=Sub Segments, parameterName=SubRingThickness, sliderMax=1}}
  MaterialExpressionShadingPathSwitch_0 {
   class=MaterialExpressionShadingPathSwitch
   inputs [
    Default {from=MaterialExpressionQualitySwitch_0, output=""}
    Deferred {from="", output=""}
    Forward {from="", output=""}
    Mobile {from=MaterialExpressionReroute_0, output=""}
   ]
   properties {}
  }
  MaterialExpressionShadingPathSwitch_1 {
   class=MaterialExpressionShadingPathSwitch
   inputs [
    Default {from=MaterialExpressionQualitySwitch_1, output=""}
    Deferred {from="", output=""}
    Forward {from="", output=""}
    Mobile {from=MaterialExpressionMultiply_41, output=""}
   ]
   properties {}
  }
  MaterialExpressionShadingPathSwitch_2 {
   class=MaterialExpressionShadingPathSwitch
   inputs [
    Default {from=MaterialExpressionQualitySwitch_2, output=""}
    Deferred {from="", output=""}
    Forward {from="", output=""}
    Mobile {from=MaterialExpressionMultiply_42, output=""}
   ]
   properties {}
  }
  MaterialExpressionShadingPathSwitch_3 {
   class=MaterialExpressionShadingPathSwitch
   inputs [
    Default {from=MaterialExpressionQualitySwitch_3, output=""}
    Deferred {from="", output=""}
    Forward {from="", output=""}
    Mobile {from=MaterialExpressionVectorParameter_6, output=RGB}
   ]
   properties {}
  }
  MaterialExpressionShadingPathSwitch_4 {
   class=MaterialExpressionShadingPathSwitch
   inputs [
    Default {from=MaterialExpressionQualitySwitch_4, output=""}
    Deferred {from="", output=""}
    Forward {from="", output=""}
    Mobile {from=MaterialExpressionNamedRerouteUsage_3, output=""}
   ]
   properties {}
  }
  MaterialExpressionStaticBool_0 {class=MaterialExpressionStaticBool}
  MaterialExpressionStaticBool_1 {class=MaterialExpressionStaticBool}
  MaterialExpressionStaticSwitchParameter_0 {
   class=MaterialExpressionStaticSwitchParameter
   inputs [
    True {from=MaterialExpressionMultiply_22, output=""}
    False {from=MaterialExpressionVectorParameter_5, output=RGB}
   ]
   properties {defaultValue=true, group=Visuals, parameterName=Use Blur}
  }
  MaterialExpressionStaticSwitchParameter_1 {
   class=MaterialExpressionStaticSwitchParameter
   inputs [
    True {from=MaterialExpressionMultiply_23, output=""}
    False {from=MaterialExpressionVectorParameter_4, output=RGB}
   ]
   properties {defaultValue=true, group=Visuals, parameterName=Use Blur}
  }
  MaterialExpressionStaticSwitchParameter_2 {
   class=MaterialExpressionStaticSwitchParameter
   inputs [
    True {from=MaterialExpressionShadingPathSwitch_1, output=""}
    False {from=MaterialExpressionNamedRerouteUsage_80, output=""}
   ]
   properties {defaultValue=true, group=Visuals, parameterName=Use Blur}
  }
  MaterialExpressionStaticSwitchParameter_3 {
   class=MaterialExpressionStaticSwitchParameter
   inputs [
    True {from=MaterialExpressionShadingPathSwitch_2, output=""}
    False {from=MaterialExpressionNamedRerouteUsage_79, output=""}
   ]
   properties {defaultValue=true, group=Visuals, parameterName=Use Blur}
  }
  MaterialExpressionStaticSwitchParameter_5 {
   class=MaterialExpressionStaticSwitchParameter
   inputs [
    True {from=MaterialExpressionLinearInterpolate_2, output=""}
    False {from=MaterialExpressionLinearInterpolate_9, output=""}
   ]
   properties {defaultValue=true, group=Visuals, parameterName=Is Mini Wheel}
  }
  MaterialExpressionStaticSwitchParameter_6 {
   class=MaterialExpressionStaticSwitchParameter
   inputs [
    True {from=MaterialExpressionNamedRerouteDeclaration_27, output=""}
    False {from=MaterialExpressionMax_0, output=""}
   ]
   properties {defaultValue=true, group=Visuals, parameterName=Is Mini Wheel}
  }
  MaterialExpressionStep_1 {
   class=MaterialExpressionStep
   inputs [
    Y {from="", output=""}
    X {from=MaterialExpressionDesaturation_0, output=""}
   ]
   properties {constY=1}
  }
  MaterialExpressionSubtract_0 {
   class=MaterialExpressionSubtract
   inputs [
    A {from=MaterialExpressionCeil_2}
    B {from=""}
   ]
  }
  MaterialExpressionSubtract_1 {
   class=MaterialExpressionSubtract
   inputs [
    A {from=MaterialExpressionNamedRerouteDeclaration_20}
    B {from=MaterialExpressionNamedRerouteUsage_17}
   ]
  }
  MaterialExpressionSubtract_10 {
   class=MaterialExpressionSubtract
   inputs [
    A {from=MaterialExpressionCeil_11}
    B {from=MaterialExpressionAdd_19}
   ]
  }
  MaterialExpressionSubtract_13 {
   class=MaterialExpressionSubtract
   inputs [
    A {from=MaterialExpressionConstant_16}
    B {from=MaterialExpressionNamedRerouteUsage_87}
   ]
  }
  MaterialExpressionSubtract_2 {
   class=MaterialExpressionSubtract
   inputs [
    A {from=MaterialExpressionCeil_3}
    B {from=MaterialExpressionAdd_2}
   ]
  }
  MaterialExpressionSubtract_3 {
   class=MaterialExpressionSubtract
   inputs [
    A {from=MaterialExpressionCeil_4}
    B {from=MaterialExpressionAdd_7}
   ]
  }
  MaterialExpressionSubtract_4 {
   class=MaterialExpressionSubtract
   inputs [
    A {from=MaterialExpressionCeil_5}
    B {from=MaterialExpressionAdd_13}
   ]
  }
  MaterialExpressionSubtract_5 {
   class=MaterialExpressionSubtract
   inputs [
    A {from=MaterialExpressionCeil_7}
    B {from=MaterialExpressionAdd_15}
   ]
  }
  MaterialExpressionSubtract_6 {
   class=MaterialExpressionSubtract
   inputs [
    A {from=MaterialExpressionCeil_1}
    B {from=MaterialExpressionAdd_3}
   ]
  }
  MaterialExpressionSubtract_7 {
   class=MaterialExpressionSubtract
   inputs [
    A {from=MaterialExpressionCeil_8}
    B {from=MaterialExpressionAdd_16}
   ]
  }
  MaterialExpressionSubtract_8 {
   class=MaterialExpressionSubtract
   inputs [
    A {from=MaterialExpressionCeil_9}
    B {from=MaterialExpressionAdd_17}
   ]
  }
  MaterialExpressionSubtract_9 {
   class=MaterialExpressionSubtract
   inputs [
    A {from=MaterialExpressionCeil_10}
    B {from=MaterialExpressionAdd_18}
   ]
  }
  MaterialExpressionTextureCoordinate_0 {class=MaterialExpressionTextureCoordinate}
  MaterialExpressionVectorParameter_0 {class=MaterialExpressionVectorParameter, properties {group=Visuals, parameterName=SlotsColor, defaultValue=<rgba 0.0365 0.0365 0.0365 0.3>}}
  MaterialExpressionVectorParameter_1 {
   class=MaterialExpressionVectorParameter
   properties {group=Visuals, parameterName=SelectedColorCenter, defaultValue=<rgba 0.375 0.375 0.375 0.7>}
  }
  MaterialExpressionVectorParameter_10 {class=MaterialExpressionVectorParameter, properties {group=SlotColors, parameterName=SlotColor1, defaultValue=<rgba 1 0 0.0157 1>}}
  MaterialExpressionVectorParameter_11 {class=MaterialExpressionVectorParameter, properties {group=SlotColors, parameterName=SlotColor2, defaultValue=<rgba 0 1 0.3979 1>}}
  MaterialExpressionVectorParameter_12 {class=MaterialExpressionVectorParameter, properties {group=SlotColors, parameterName=SlotColor3, defaultValue=<rgba 0 0.9475 1 1>}}
  MaterialExpressionVectorParameter_13 {class=MaterialExpressionVectorParameter, properties {group=SlotColors, parameterName=SlotColor4, defaultValue=<rgba 1 0.8817 0 1>}}
  MaterialExpressionVectorParameter_14 {class=MaterialExpressionVectorParameter, properties {group=SlotColors, parameterName=SlotColor5, defaultValue=<rgba 1 0 0.7861 1>}}
  MaterialExpressionVectorParameter_15 {class=MaterialExpressionVectorParameter, properties {group=SlotColors, parameterName=SlotColor6, defaultValue=<rgba 0.4585 0 1 1>}}
  MaterialExpressionVectorParameter_16 {class=MaterialExpressionVectorParameter, properties {group=SlotColors, parameterName=SlotColor7, defaultValue=<rgba 1 0.3075 0 1>}}
  MaterialExpressionVectorParameter_2 {
   class=MaterialExpressionVectorParameter
   properties {group=Visuals, parameterName=SelectedColorOuter, defaultValue=<rgba 0.319 0.319 0.319 0.3>}
  }
  MaterialExpressionVectorParameter_3 {class=MaterialExpressionVectorParameter, properties {group=Visuals, parameterName=SubSegmentColor, defaultValue=<rgba 1 1 1 1>}}
  MaterialExpressionVectorParameter_4 {
   class=MaterialExpressionVectorParameter
   properties {group=Visuals, parameterName=CenterBackgroundColor, defaultValue=<rgba 0.5972 0.5972 0.5972 1>}
  }
  MaterialExpressionVectorParameter_5 {
   class=MaterialExpressionVectorParameter
   properties {group=Visuals, parameterName=DonutBackgroundColor, defaultValue=<rgba 0 0.0123 0.025 1>}
  }
  MaterialExpressionVectorParameter_6 {
   class=MaterialExpressionVectorParameter
   properties {group=Visuals, parameterName=DonutBackgroundColorLowEnd, defaultValue=<rgba 0 0.0123 0.025 1>}
  }
  MaterialExpressionVectorParameter_7 {
   class=MaterialExpressionVectorParameter
   properties {group=Visuals, parameterName=SubSegmentBackgroundColor, defaultValue=<rgba 0.6254 1 0 0.5>}
  }
  MaterialExpressionVectorParameter_8 {class=MaterialExpressionVectorParameter, properties {group=SlotColors, parameterName=SlotColor0, defaultValue=<rgba 0.0011 0.3486 1 1>}}
 }
 properties {
  bAutomaticallySetUsageInEditor=false
  bUsedWithStaticMesh=false
  blendMode=BLEND_Translucent
  materialDomain=MD_UI
  usedShadingModels=Default Lit
 }
}
