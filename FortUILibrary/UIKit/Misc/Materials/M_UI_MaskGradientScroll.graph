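# Translucent UI material that builds an adjustable linear gradient, named as a scrolling gradient mask for UI.
# MF_UI_LinearGradient takes texture coordinates plus GradientStart, GradientEnd, Rotation and GradientOffset scalar parameters to shape and position the gradient.
# The graph also holds a GetSlatePost1 call for sampling a Slate post buffer.
# No time-based node appears in the graph, so any movement depends on changing the parameters or on the gradient function.
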Material /FortUILibrary/UIKit/Misc/Materials/M_UI_MaskGradientScroll
graph {
 properties {
  bUsedWithStaticMesh=false
  blendMode=BLEND_Translucent
  materialDomain=MD_UI
  pixelDepthOffsetMode=PDOM_Legacy
  usedShadingModels=Default Lit
 }
 nodes [
  MaterialFunctionCall(
   UVs=TextureCoordinate()
   GradientStart=ScalarParameter() {parameterName=GradientStart}
   GradientEnd=ScalarParameter() {defaultValue=1, parameterName=GradientEnd}
   RotationAngle=ScalarParameter() {parameterName=Rotation}
   GradientOffset=ScalarParameter() {parameterName=GradientOffset}
  ) {materialFunction=&MF_UI_LinearGradient}
  MaterialFunctionCall!("UVs (Optional)") {materialFunction=/Engine/Functions/UserInterface/GetSlatePost1}
 ]
}
